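# Material function for the American Football universal vehicle skin, outputting material attributes with animated patterns in team, accent, paint and trim colors.
# Panned and rotated UV1 samples of a packed RGB texture combine with procedural repeating stripes into layered color masks, with a Painted static switch changing the color assignment.
# The combined mask also drives surface response, blending Fresnel-based jersey and rubber finishes alongside metallic, roughness and a jersey-weave normal.
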MaterialFunction /VehicleCosmetics/UniversalSkins/AmericanFootball/MF_Vehicle_Skin_AmericanFootball
graph {nodes [
  Composite() {subgraphName=Vectors}
  n1 = NamedRerouteUsage()
  n2 = SetMaterialAttributes!(
   MaterialAttributes
   BaseColor=NamedRerouteUsage()
   Metallic=LinearInterpolate(Alpha=@n1) {constA=0.75, constB=0}
   Roughness=LinearInterpolate(Alpha=@n1) {constA=0.4, constB=0.2}
   CustomData0
   CustomData1
   Normal=NamedRerouteUsage()
   ClearCoatBottomNormal
   EmissiveColor
  ) {attributeSetTypes [
    #1
    #2
    #3
    #4
    #5
    #6
    #7
    #8
   ]}
  FunctionOutput(None=@n2)
  n3 = TextureSampleParameter2D(UVs=TextureCoordinate() {coordinateIndex=1, uTiling=1.5, vTiling=3}) {parameterName=Normal, samplerType=SAMPLERTYPE_Normal, texture=/VehicleCosmetics/UniversalSkins/ColorRush/N_jerseyweave}
  n4 = Constant3Vector() {constant=<rgba 0 0 1 0>}
  n5 = NamedRerouteUsage()
  n6 = ScalarParameter() {defaultValue=0.5, parameterName=Normal Intensity}
  n7 = Constant() {r=1}
  NamedRerouteDeclaration(None=Multiply(
    A=LinearInterpolate(A=@n3.RGB, B=@n4, Alpha=@n5)
    B=MaterialFunctionCall(R=@n6, G=@n6, B=@n7) {materialFunction=&AppendMany}.RGB
   )) {name=Normal, nodeColor=<rgba 0.6172 0 1 1>}
  n8 = Constant() {r=0.4}
  n9 = Power(Base=OneMinus(None=MaterialFunctionCall!(Fresnel Power=@n8, Normal=@n4) {materialFunction=/VehicleCosmetics/Materials/MaterialFunctions/MF_Fresnel_Prime}.Result)) {constExponent=15}
  n10 = Saturate(None=Multiply(A=Saturate(None=@n9)) {constB=2.5})
  NamedRerouteDeclaration(None=@n10) {name=highlight, nodeColor=<rgba 0 0.5078 1 1>}
  n11 = VectorParameter() {group=Color, parameterName=BaseColor, sortPriority=1, defaultValue=<rgba 0.4313 0.4313 0.4313 1>}
  MaterialExpressionReroute_23 = Reroute(None=@n11.RGB)
  NamedRerouteDeclaration(None=@MaterialExpressionReroute_23.TeamColor) {name=TeamColor, nodeColor=<rgba 0.9219 0 1 1>}
  n12 = VectorParameter() {group=Color, parameterName=AccentColor, sortPriority=2, defaultValue=<rgba 0 0.2783 0.9739 1>}
  MaterialExpressionReroute_24 = Reroute(None=@n12.RGB)
  NamedRerouteDeclaration(None=@MaterialExpressionReroute_24.AccentColor) {name=AccentColor, nodeColor=<rgba 0.9219 0 1 1>}
  n13 = VectorParameter() {group=Color, parameterName=PaintColor, sortPriority=3, defaultValue=<rgba 0 0 0 1>}
  MaterialExpressionReroute_25 = Reroute(None=@n13.RGB)
  NamedRerouteDeclaration(None=@MaterialExpressionReroute_25.PaintColor) {name=PaintColor, nodeColor=<rgba 0.9219 0 1 1>}
  NamedRerouteDeclaration(None=Time() {bOverride_Period=true, period=500}) {name=time, nodeColor=<rgba 0.8516 1 0 1>}
  n14 = Constant() {r=5}
  n15 = Saturate(None=Add(A=MaterialFunctionCall(Fresnel Power=@n14, Normal=@n3.RGB) {materialFunction=&MF_Fresnel_Prime}.Result) {constB=0.125})
  n16 = Constant() {r=1}
  n17 = Power(Base=OneMinus(None=MaterialFunctionCall(Fresnel Power=@n16, Normal=@n3.RGB) {materialFunction=&MF_Fresnel_Prime}.Result)) {constExponent=15}
  n18 = Clamp(None=Multiply(A=Saturate(None=@n17)) {constB=2}) {maxDefault=1.5}
  NamedRerouteDeclaration(None=Saturate(None=Add(A=@n15, B=@n18))) {name=JerseyFinish, nodeColor=<rgba 0 1 0.3203 1>}
  n19 = Constant() {r=1.5}
  n20 = Saturate(None=ConstantBiasScale(None=MaterialFunctionCall(Fresnel Power=@n19, Normal=@n4) {materialFunction=&MF_Fresnel_Prime}.Result) {bias=0.15, scale=2.5})
  NamedRerouteDeclaration(None=Saturate(None=Max(A=@n20, B=@n10))) {name=rubberfinish, nodeColor=<rgba 1 0.7969 0 1>}
  n21 = NamedRerouteUsage()
  n22 = NamedRerouteUsage()
  n23 = TextureCoordinate() {coordinateIndex=1}
  n24 = ComponentMask(None=@n23) {g=true}
  n25 = Saturate(None=ConstantBiasScale(None=OneMinus(None=@n24)) {bias=-0.5, scale=3})
  n26 = Multiply(A=Rotator(Coordinate=@n23, Time=Constant() {r=0.15}) {speed=1}, B=Constant2Vector() {g=0.4, r=7})
  n27 = NamedRerouteUsage()
  n28 = TextureSampleParameter2D(UVs=BumpOffset(Coordinate=Panner(Coordinate=@n26, Time=@n27) {speedX=-0.001, speedY=-0.08}, Height=Constant() {r=0.4}) {heightRatio=1}) {parameterName=RGB, texture=&/T_AmericanFootball_RGB}
  n29 = ComponentMask(None=Rotator(Coordinate=TextureCoordinate() {coordinateIndex=1, uTiling=0, vTiling=-125}, Time=Constant() {r=-0.8}) {speed=1}) {g=true}
  n30 = Subtract(A=@n29, B=Floor(None=@n29))
  n31 = Add(A=Saturate(None=Multiply(A=@n28.R) {constB=5}), B=Saturate(None=Subtract(A=@n30) {constB=0.05}))
  n32 = Multiply(A=Saturate(None=Multiply(A=@n28.R) {constB=3}), B=Saturate(None=ConstantBiasScale(None=@n31) {bias=-0.85, scale=15}))
  n33 = ConstantBiasScale(None=OneMinus(None=@n32)) {bias=-0.7, scale=3}
  n34 = LinearInterpolate(
   A=NamedRerouteUsage()
   B=StaticSwitchParameter(True=LinearInterpolate(A=@n21, B=@n22, Alpha=@n25), False=@n22) {parameterName=Painted}
   Alpha=LinearInterpolate(A=@n32, B=Saturate(None=@n33), Alpha=@n25)
  )
  n35 = Constant() {r=0.1}
  n36 = NamedRerouteUsage()
  n37 = NamedRerouteUsage()
  n38 = TextureCoordinate() {coordinateIndex=1}
  n39 = Constant2Vector() {g=0.25, r=1}
  n40 = TextureSampleParameter2D(UVs=Panner(Coordinate=Multiply(A=@n38, B=@n39), Time=NamedRerouteUsage()) {speedY=-0.2}) {parameterName=RGB, texture=&T_AmericanFootball_RGB}
  n41 = Constant2Vector() {g=0.5, r=1}
  n42 = TextureSampleParameter2D(UVs=Add(A=Multiply(A=@n38, B=@n41), B=Constant2Vector() {g=-0.1})) {parameterName=RGB, texture=&T_AmericanFootball_RGB}
  n43 = ComponentMask(None=Panner(Coordinate=TextureCoordinate() {coordinateIndex=1, uTiling=11, vTiling=-31}, Time=NamedRerouteUsage()) {speedY=-2}) {g=true}
  n44 = ConstantBiasScale(None=Subtract(A=@n43, B=Floor(None=@n43))) {bias=-0.5, scale=2}
  n45 = Saturate(None=Subtract(A=Abs(None=@n44)) {constB=0.05})
  n46 = Add(A=Saturate(None=ConstantBiasScale(None=@n40.G) {bias=-0.01, scale=2}), B=Multiply(A=@n42.A, B=@n45))
  n47 = Multiply(A=Saturate(None=Multiply(A=@n40.G) {constB=15}), B=Saturate(None=ConstantBiasScale(None=@n46) {bias=-0.9, scale=15}))
  n48 = LinearInterpolate(
   A=LinearInterpolate(A=@n34, B=Desaturation(None=@n34, Fraction=@n35), Alpha=NamedRerouteUsage())
   B=StaticSwitchParameter(True=@n36, False=@n37) {parameterName=Painted}
   Alpha=@n47
  )
  n49 = NamedRerouteUsage()
  n50 = ConstantBiasScale(None=ComponentMask(None=@n38) {r=true}) {bias=-0.5, scale=2}
  n51 = Saturate(None=ConstantBiasScale(None=Abs(None=@n50)) {bias=-0.35, scale=55})
  n52 = LinearInterpolate(A=@n40.R, B=@n42.R, Alpha=@n51)
  n53 = Add(A=Saturate(None=Multiply(A=@n52) {constB=3}), B=Multiply(A=@n42.B, B=@n45))
  n54 = Multiply(A=Saturate(None=Multiply(A=@n52) {constB=51}), B=Saturate(None=ConstantBiasScale(None=@n53) {bias=-0.6, scale=15}))
  n55 = Rotator(Coordinate=TextureCoordinate() {coordinateIndex=1}, Time=Constant() {r=0.5}) {speed=1}
  n56 = ComponentMask(None=@n55) {g=true}
  n57 = Constant2Vector() {g=31, r=1}
  n58 = ComponentMask(None=Panner(Coordinate=Multiply(A=@n55, B=@n57), Time=NamedRerouteUsage()) {speedY=1}) {g=true}
  n59 = ConstantBiasScale(None=Subtract(A=@n58, B=Floor(None=@n58))) {bias=-0.5, scale=2}
  n60 = Saturate(None=ConstantBiasScale(None=Abs(None=@n59)) {bias=-0.05, scale=0.2})
  n61 = Saturate(None=ConstantBiasScale(None=Add(A=@n56, B=@n60)) {bias=-0.8, scale=1000})
  n62 = Max(A=@n54, B=@n61)
  n63 = LinearInterpolate(A=@n36, B=@n37, Alpha=@n51)
  n64 = OneMinus(None=@n56)
  n65 = Saturate(None=Multiply(A=@n61, B=Multiply(A=@n64) {constB=7}) {constB=7})
  n66 = Saturate(None=Multiply(A=@n54, B=@n65))
  n67 = Saturate(None=Multiply(A=@n47, B=@n65))
  n68 = LinearInterpolate(
   A=LinearInterpolate(A=LinearInterpolate(A=@n48, B=@n49, Alpha=@n62), B=StaticSwitchParameter(True=@n63, False=@n37) {parameterName=Painted}, Alpha=@n66)
   B=@n37
   Alpha=@n67
  )
  n69 = Saturate(None=Subtract(A=Max(A=@n62, B=@n47), B=Max(A=@n66, B=@n67))) {desc=env}
  n70 = LinearInterpolate(A=NamedRerouteUsage(), B=NamedRerouteUsage(), Alpha=@n69)
  n71 = Multiply(A=NamedRerouteUsage(), B=@n69)
  n72 = Constant() {r=0.4}
  NamedRerouteDeclaration(None=LinearInterpolate(A=Multiply(A=@n68, B=@n70), B=Constant3Vector() {constant=<rgba 1 1 1 0>}, Alpha=Multiply(A=@n71, B=@n72))) {name=basecolor, nodeColor=<rgba 0 1 0.5547 1>}
  NamedRerouteDeclaration(None=@n69) {name=env mask, nodeColor=<rgba 1 0.2344 0 1>}
  PinBase()
  n73 = VectorParameter() {group=Color, parameterName=TrimColor, sortPriority=4, defaultValue=<rgba 0.03 0.03 0.03 0>}
  PinBase!(TeamColor=@n11.RGB, AccentColor=@n12.RGB, PaintColor=@n13.RGB, TrimColor=@n73.RGB) {reroutePins [
    TeamColor {expression=&self:MaterialExpressionReroute_23}
    AccentColor {expression=&self:MaterialExpressionReroute_24}
    PaintColor {expression=&self:MaterialExpressionReroute_25}
    TrimColor {expression=&self:MaterialExpressionReroute_43}
   ]}
  MaterialExpressionReroute_43 = Reroute(None=@n73.RGB)
  n74 = TextureCoordinate() {coordinateIndex=1}
  n75 = Constant() {r=0.5}
  n76 = ConstantBiasScale(None=ComponentMask(None=Rotator(Coordinate=@n74, Time=@n75) {speed=1}) {g=true}) {bias=-0.475, scale=6}
  StaticSwitchParameter(True=LinearInterpolate(A=@n34, B=@n21, Alpha=Saturate(None=@n76)), False=@n34) {parameterName=Painted}
 ]}
