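# Parent material function for the American football "Pigskin" universal vehicle skin, outputting a full set of material attributes.
# Composite subgraphs build body textures, finish, details, roughness, base color, clear coat and emissive, blending team, accent and paint values through skin mask channels.
# Additional layers combine texture masks with bump offset, panning, Fresnel and camera and reflection vectors for depth and view-dependent effects.
# Subgraphs marked as legacy Rocket League features handle sharp masks.
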MaterialFunction /VehicleCosmetics/UniversalSkins/AmericanFootball/MF_Vehicle_Skin_Pigskin_Parent
graph {expressions {
  MaterialExpressionAbs_0 {class=MaterialExpressionAbs, inputs [None {from=MaterialExpressionAdd_0, output=""}], properties {}}
  MaterialExpressionAbs_2 {class=MaterialExpressionAbs, inputs [None {from=MaterialExpressionClamp_11, output=""}], properties {}}
  MaterialExpressionAdd_0 {
   class=MaterialExpressionAdd
   inputs [
    A {from=MaterialExpressionConstant_0}
    B {from=MaterialExpressionPower_1}
   ]
  }
  MaterialExpressionAdd_1 {
   class=MaterialExpressionAdd
   inputs [
    A {from=MaterialExpressionComponentMask_1}
    B {from=MaterialExpressionComponentMask_1}
   ]
  }
  MaterialExpressionAdd_10 {
   class=MaterialExpressionAdd
   inputs [
    A {from=MaterialExpressionMultiply_59}
    B {from=MaterialExpressionMultiply_60}
   ]
  }
  MaterialExpressionAdd_13 {
   class=MaterialExpressionAdd
   inputs [
    A {from=MaterialExpressionMultiply_73}
    B {from=MaterialExpressionLinearInterpolate_35}
   ]
  }
  MaterialExpressionAdd_14 {
   class=MaterialExpressionAdd
   inputs [
    A {from=MaterialExpressionLinearInterpolate_42}
    B {from=MaterialExpressionStaticSwitchParameter_4}
   ]
  }
  MaterialExpressionAdd_15 {
   class=MaterialExpressionAdd
   inputs [
    A {from=MaterialExpressionMultiply_56}
    B {from=MaterialExpressionMultiply_57}
   ]
  }
  MaterialExpressionAdd_16 {
   class=MaterialExpressionAdd
   inputs [
    A {from=MaterialExpressionMultiply_61}
    B {from=MaterialExpressionMultiply_62}
   ]
  }
  MaterialExpressionAdd_2 {
   class=MaterialExpressionAdd
   inputs [
    A {from=MaterialExpressionMultiply_36}
    B {from=MaterialExpressionAppendVector_18}
   ]
  }
  MaterialExpressionAdd_22 {
   class=MaterialExpressionAdd
   inputs [
    A {from=MaterialExpressionMultiply_75}
    B {from=MaterialExpressionMultiply_55}
   ]
  }
  MaterialExpressionAdd_23 {
   class=MaterialExpressionAdd
   inputs [
    A {from=MaterialExpressionLinearInterpolate_37}
    B {from=MaterialExpressionMultiply_55}
   ]
  }
  MaterialExpressionAdd_26 {
   class=MaterialExpressionAdd
   inputs [
    A {from=MaterialExpressionLinearInterpolate_38}
    B {from=MaterialExpressionMultiply_77}
   ]
  }
  MaterialExpressionAdd_3 {
   class=MaterialExpressionAdd
   inputs [
    A {from=MaterialExpressionReroute_64}
    B {from=MaterialExpressionAdd_16}
   ]
  }
  MaterialExpressionAdd_4 {
   class=MaterialExpressionAdd
   inputs [
    A {from=MaterialExpressionAdd_1}
    B {from=MaterialExpressionPower_5}
   ]
  }
  MaterialExpressionAdd_9 {
   class=MaterialExpressionAdd
   inputs [
    A {from=MaterialExpressionComponentMask_6}
    B {from=MaterialExpressionComponentMask_33}
   ]
  }
  MaterialExpressionAppendVector_0 {
   class=MaterialExpressionAppendVector
   inputs [
    A {from=MaterialExpressionTextureSampleParameter2D_0, output=B}
    B {from=MaterialExpressionTextureSampleParameter2D_0, output=B}
   ]
  }
  MaterialExpressionAppendVector_1 {
   class=MaterialExpressionAppendVector
   inputs [
    A {from=MaterialExpressionComponentMask_4, output=""}
    B {from=MaterialExpressionComponentMask_5, output=""}
   ]
  }
  MaterialExpressionAppendVector_10 {
   class=MaterialExpressionAppendVector
   inputs [
    A {from=MaterialExpressionVectorParameter_10, output=R}
    B {from=MaterialExpressionVectorParameter_10, output=R}
   ]
  }
  MaterialExpressionAppendVector_11 {
   class=MaterialExpressionAppendVector
   inputs [
    A {from=MaterialExpressionSubtract_17, output=""}
    B {from=MaterialExpressionSubtract_18, output=""}
   ]
  }
  MaterialExpressionAppendVector_12 {
   class=MaterialExpressionAppendVector
   inputs [
    A {from=MaterialExpressionVectorParameter_11, output=R}
    B {from=MaterialExpressionVectorParameter_11, output=R}
   ]
  }
  MaterialExpressionAppendVector_13 {
   class=MaterialExpressionAppendVector
   inputs [
    A {from=MaterialExpressionVectorParameter_12, output=R}
    B {from=MaterialExpressionVectorParameter_12, output=R}
   ]
  }
  MaterialExpressionAppendVector_14 {
   class=MaterialExpressionAppendVector
   inputs [
    A {from=MaterialExpressionVectorParameter_14, output=R}
    B {from=MaterialExpressionVectorParameter_14, output=R}
   ]
  }
  MaterialExpressionAppendVector_15 {
   class=MaterialExpressionAppendVector
   inputs [
    A {from=MaterialExpressionSubtract_25, output=""}
    B {from=MaterialExpressionSubtract_26, output=""}
   ]
  }
  MaterialExpressionAppendVector_16 {
   class=MaterialExpressionAppendVector
   inputs [
    A {from=MaterialExpressionSubtract_27, output=""}
    B {from=MaterialExpressionSubtract_28, output=""}
   ]
  }
  MaterialExpressionAppendVector_17 {
   class=MaterialExpressionAppendVector
   inputs [
    A {from=MaterialExpressionVectorParameter_15, output=R}
    B {from=MaterialExpressionVectorParameter_15, output=R}
   ]
  }
  MaterialExpressionAppendVector_18 {
   class=MaterialExpressionAppendVector
   inputs [
    A {from=MaterialExpressionVectorParameter_15, output=B}
    B {from=MaterialExpressionVectorParameter_15, output=B}
   ]
  }
  MaterialExpressionAppendVector_2 {
   class=MaterialExpressionAppendVector
   inputs [
    A {from=MaterialExpressionConstant4Vector_1, output=R}
    B {from=MaterialExpressionConstant4Vector_1, output=R}
   ]
  }
  MaterialExpressionAppendVector_3 {
   class=MaterialExpressionAppendVector
   inputs [
    A {from=MaterialExpressionComponentMask_28, output=""}
    B {from=MaterialExpressionComponentMask_27, output=""}
   ]
  }
  MaterialExpressionAppendVector_4 {
   class=MaterialExpressionAppendVector
   inputs [
    A {from=MaterialExpressionConstant4Vector_2, output=R}
    B {from=MaterialExpressionConstant4Vector_2, output=R}
   ]
  }
  MaterialExpressionAppendVector_5 {
   class=MaterialExpressionAppendVector
   inputs [
    A {from=MaterialExpressionConstant4Vector_2, output=B}
    B {from=MaterialExpressionConstant4Vector_2, output=B}
   ]
  }
  MaterialExpressionAppendVector_6 {
   class=MaterialExpressionAppendVector
   inputs [
    A {from=MaterialExpressionConstant4Vector_3, output=R}
    B {from=MaterialExpressionConstant4Vector_3, output=R}
   ]
  }
  MaterialExpressionAppendVector_7 {
   class=MaterialExpressionAppendVector
   inputs [
    A {from=MaterialExpressionConstant4Vector_3, output=B}
    B {from=MaterialExpressionConstant4Vector_3, output=B}
   ]
  }
  MaterialExpressionAppendVector_8 {
   class=MaterialExpressionAppendVector
   inputs [
    A {from=MaterialExpressionVectorParameter_8, output=R}
    B {from=MaterialExpressionVectorParameter_8, output=R}
   ]
  }
  MaterialExpressionAppendVector_9 {
   class=MaterialExpressionAppendVector
   inputs [
    A {from=MaterialExpressionSubtract_10, output=""}
    B {from=MaterialExpressionSubtract_11, output=""}
   ]
  }
  MaterialExpressionBumpOffset_1 {
   class=MaterialExpressionBumpOffset
   inputs [
    Coordinate {from=MaterialExpressionAdd_16, output=""}
    Height {from=MaterialExpressionConstant_3, output=""}
    HeightRatioInput {from=MaterialExpressionConstant_4, output=""}
   ]
   properties {}
  }
  MaterialExpressionCameraVectorWS_0 {class=MaterialExpressionCameraVectorWS, inputs [], properties {}}
  MaterialExpressionCameraVectorWS_1 {class=MaterialExpressionCameraVectorWS, inputs [], properties {}}
  MaterialExpressionCameraVectorWS_2 {class=MaterialExpressionCameraVectorWS, inputs [], properties {}}
  MaterialExpressionChannelMaskParameter_1 {
   class=MaterialExpressionChannelMaskParameter
   inputs [None {from=MaterialExpressionAppendVector_1, output=""}]
   properties {
    desc="R: SkinMask - Red; G: SkinMask - Green; B: SkinMask - Blue; A: BodyTextures - Blue"
    group=Color
    maskChannel=Alpha
    parameterName=EmissiveMaskSelector
    sortPriority=6
    defaultValue=<rgba 0 0 0 1>
   }
  }
  MaterialExpressionClamp_1 {
   class=MaterialExpressionClamp
   inputs [
    None {from=MaterialExpressionPower_7, output=""}
    Min {from=MaterialExpressionConstant_6, output=""}
    Max {from="", output=""}
   ]
   properties {clampMode=CMODE_ClampMin, minDefault=0.5}
  }
  MaterialExpressionClamp_10 {
   class=MaterialExpressionClamp
   inputs [
    None {from=MaterialExpressionConstantBiasScale_7, output=""}
    Min {from="", output=""}
    Max {from="", output=""}
   ]
   properties {}
  }
  MaterialExpressionClamp_11 {
   class=MaterialExpressionClamp
   inputs [
    None {from=MaterialExpressionMax_2, output=""}
    Min {from=MaterialExpressionConstant_20, output=""}
    Max {from=MaterialExpressionMultiply_14, output=""}
   ]
   properties {}
  }
  MaterialExpressionClamp_12 {
   class=MaterialExpressionClamp
   inputs [
    None {from=MaterialExpressionAbs_2, output=""}
    Min {from="", output=""}
    Max {from="", output=""}
   ]
   properties {}
  }
  MaterialExpressionClamp_14 {
   class=MaterialExpressionClamp
   inputs [
    None {from=MaterialExpressionMax_1, output=""}
    Min {from="", output=""}
    Max {from="", output=""}
   ]
   properties {}
  }
  MaterialExpressionClamp_15 {
   class=MaterialExpressionClamp
   inputs [
    None {from=MaterialExpressionComponentMask_32, output=""}
    Min {from="", output=""}
    Max {from="", output=""}
   ]
   properties {}
  }
  MaterialExpressionClamp_16 {
   class=MaterialExpressionClamp
   inputs [
    None {from=MaterialExpressionConstantBiasScale_10, output=""}
    Min {from="", output=""}
    Max {from="", output=""}
   ]
   properties {}
  }
  MaterialExpressionClamp_17 {
   class=MaterialExpressionClamp
   inputs [
    None {from=MaterialExpressionMultiply_66, output=""}
    Min {from="", output=""}
    Max {from="", output=""}
   ]
   properties {}
  }
  MaterialExpressionClamp_18 {
   class=MaterialExpressionClamp
   inputs [
    None {from=MaterialExpressionConstantBiasScale_12, output=""}
    Min {from="", output=""}
    Max {from="", output=""}
   ]
   properties {}
  }
  MaterialExpressionClamp_3 {
   class=MaterialExpressionClamp
   inputs [
    None {from=MaterialExpressionMultiply_53, output=""}
    Min {from="", output=""}
    Max {from="", output=""}
   ]
   properties {}
  }
  MaterialExpressionClamp_4 {
   class=MaterialExpressionClamp
   inputs [
    None {from=MaterialExpressionDesaturation_0, output=""}
    Min {from="", output=""}
    Max {from="", output=""}
   ]
   properties {}
  }
  MaterialExpressionClamp_5 {
   class=MaterialExpressionClamp
   inputs [
    None {from=MaterialExpressionCameraVectorWS_0, output=""}
    Min {from="", output=""}
    Max {from="", output=""}
   ]
   properties {}
  }
  MaterialExpressionClamp_6 {
   class=MaterialExpressionClamp
   inputs [
    None {from=MaterialExpressionReflectionVectorWS_0, output=""}
    Min {from="", output=""}
    Max {from="", output=""}
   ]
   properties {}
  }
  MaterialExpressionClamp_7 {
   class=MaterialExpressionClamp
   inputs [
    None {from=MaterialExpressionConstantBiasScale_5, output=""}
    Min {from="", output=""}
    Max {from="", output=""}
   ]
   properties {}
  }
  MaterialExpressionClamp_8 {
   class=MaterialExpressionClamp
   inputs [
    None {from=MaterialExpressionConstantBiasScale_8, output=""}
    Min {from="", output=""}
    Max {from="", output=""}
   ]
   properties {}
  }
  MaterialExpressionClamp_9 {
   class=MaterialExpressionClamp
   inputs [
    None {from=MaterialExpressionConstantBiasScale_9, output=""}
    Min {from="", output=""}
    Max {from="", output=""}
   ]
   properties {}
  }
  MaterialExpressionComponentMask_0 {class=MaterialExpressionComponentMask, inputs [None {from=MaterialExpressionReroute_30, output=SkinMaskTexture}], properties {r=true}}
  MaterialExpressionComponentMask_1 {class=MaterialExpressionComponentMask, inputs [None {from=MaterialExpressionAbs_0}], properties {g=true}}
  MaterialExpressionComponentMask_15 {class=MaterialExpressionComponentMask, inputs [None {from=MaterialExpressionCameraVectorWS_1}], properties {r=true}}
  MaterialExpressionComponentMask_16 {class=MaterialExpressionComponentMask, inputs [None {from=MaterialExpressionTransform_0}], properties {b=true}}
  MaterialExpressionComponentMask_17 {class=MaterialExpressionComponentMask, inputs [None {from=MaterialExpressionSubtract_9}], properties {r=true}}
  MaterialExpressionComponentMask_18 {class=MaterialExpressionComponentMask, inputs [None {from=MaterialExpressionSubtract_9}], properties {g=true}}
  MaterialExpressionComponentMask_2 {class=MaterialExpressionComponentMask, inputs [None {from=MaterialExpressionNamedRerouteUsage_34}], properties {r=true}}
  MaterialExpressionComponentMask_20 {class=MaterialExpressionComponentMask, inputs [None {from=MaterialExpressionSubtract_14}], properties {r=true}}
  MaterialExpressionComponentMask_21 {class=MaterialExpressionComponentMask, inputs [None {from=MaterialExpressionSubtract_14}], properties {g=true}}
  MaterialExpressionComponentMask_22 {class=MaterialExpressionComponentMask, inputs [None {from=MaterialExpressionSubtract_19}], properties {r=true}}
  MaterialExpressionComponentMask_23 {class=MaterialExpressionComponentMask, inputs [None {from=MaterialExpressionSubtract_19}], properties {g=true}}
  MaterialExpressionComponentMask_24 {class=MaterialExpressionComponentMask, inputs [None {from=MaterialExpressionSubtract_22}], properties {r=true}}
  MaterialExpressionComponentMask_25 {class=MaterialExpressionComponentMask, inputs [None {from=MaterialExpressionSubtract_22}], properties {g=true}}
  MaterialExpressionComponentMask_27 {class=MaterialExpressionComponentMask, inputs [None {from=MaterialExpressionPanner_8}], properties {r=true}}
  MaterialExpressionComponentMask_28 {class=MaterialExpressionComponentMask, inputs [None {from=MaterialExpressionPanner_8}], properties {g=true}}
  MaterialExpressionComponentMask_29 {class=MaterialExpressionComponentMask, inputs [None {from=MaterialExpressionClamp_5}], properties {b=true, g=true}}
  MaterialExpressionComponentMask_3 {class=MaterialExpressionComponentMask, inputs [None {from=MaterialExpressionNamedRerouteUsage_35}], properties {g=true}}
  MaterialExpressionComponentMask_30 {class=MaterialExpressionComponentMask, inputs [None {from=MaterialExpressionClamp_6}], properties {b=true, g=true}}
  MaterialExpressionComponentMask_31 {class=MaterialExpressionComponentMask, inputs [None {from=MaterialExpressionClamp_6}], properties {r=true}}
  MaterialExpressionComponentMask_32 {class=MaterialExpressionComponentMask, inputs [None {from=MaterialExpressionClamp_5}], properties {r=true}}
  MaterialExpressionComponentMask_33 {class=MaterialExpressionComponentMask, inputs [None {from=MaterialExpressionMultiply_58}], properties {g=true}}
  MaterialExpressionComponentMask_4 {
   class=MaterialExpressionComponentMask
   inputs [None {from=MaterialExpressionReroute_55, output=SkinMaskTexture}]
   properties {b=true, g=true, r=true}
  }
  MaterialExpressionComponentMask_5 {class=MaterialExpressionComponentMask, inputs [None {from=MaterialExpressionSaturate_4}], properties {r=true}}
  MaterialExpressionComponentMask_6 {class=MaterialExpressionComponentMask, inputs [None {from=MaterialExpressionMultiply_58}], properties {r=true}}
  MaterialExpressionComposite_0 {
   class=MaterialExpressionComposite
   inputs [SkinMaskTexture {from=MaterialExpressionNamedRerouteUsage_12, output=""}]
   properties {subgraphName=BodyFinish}
  }
  MaterialExpressionComposite_1 {
   class=MaterialExpressionComposite
   inputs [
    UV {from=MaterialExpressionNamedRerouteUsage_25, output=""}
    UV Detail {from=MaterialExpressionNamedRerouteUsage_26, output=""}
    UV BottomNormal {from=MaterialExpressionNamedRerouteUsage_21, output=""}
    LOD {from=MaterialExpressionScalarParameter_1, output=""}
    BottomNormalFlatness {from=MaterialExpressionScalarParameter_14, output=""}
   ]
   properties {subgraphName=BodyTextures}
  }
  MaterialExpressionComposite_10 {
   class=MaterialExpressionComposite
   inputs [
    BaseColorInput {from=MaterialExpressionReroute_28, output=BaseColor}
    TextureInput {from=MaterialExpressionNamedRerouteUsage_27, output=""}
   ]
   properties {desc=Legacy - Rocket League Feature, subgraphName=SharpMasksBaseColorMultiply}
  }
  MaterialExpressionComposite_2 {
   class=MaterialExpressionComposite
   inputs [
    DetailBaseTexture {from=MaterialExpressionNamedRerouteUsage_29, output=""}
    DetailAccentTexture {from=MaterialExpressionNamedRerouteUsage_30, output=""}
    SkinMaskTexture {from=MaterialExpressionNamedRerouteUsage_18, output=""}
   ]
   properties {subgraphName=Details}
  }
  MaterialExpressionComposite_3 {
   class=MaterialExpressionComposite
   inputs [
    Roughness {from=MaterialExpressionNamedRerouteUsage_5, output=""}
    SkinMask {from=MaterialExpressionNamedRerouteUsage_4, output=""}
   ]
   properties {subgraphName=Roughness}
  }
  MaterialExpressionComposite_4 {
   class=MaterialExpressionComposite
   inputs [
    TeamColor {from="", output=""}
    AccentColor {from="", output=""}
    PaintColor {from=MaterialExpressionNamedRerouteUsage_10, output=""}
    SkinMaskTexture {from=MaterialExpressionNamedRerouteUsage_7, output=""}
   ]
   properties {subgraphName=BaseColor}
  }
  MaterialExpressionComposite_5 {
   class=MaterialExpressionComposite
   inputs [
    BaseColor {from=MaterialExpressionReroute_26, output=BaseColor}
    BaseColorTexture {from=MaterialExpressionNamedRerouteUsage_15, output=""}
    SkinMaskTexture {from=MaterialExpressionNamedRerouteUsage_14, output=""}
   ]
   properties {subgraphName=UseBaseColorTexture}
  }
  MaterialExpressionComposite_6 {class=MaterialExpressionComposite, inputs [], properties {subgraphName=Vectors}}
  MaterialExpressionComposite_7 {
   class=MaterialExpressionComposite
   inputs [
    SkinMaskTexture {from=MaterialExpressionNamedRerouteUsage_13, output=""}
    BottomNormalBaseTexture {from=MaterialExpressionNamedRerouteUsage_11, output=""}
    BottomNormalAccentTexture {from=MaterialExpressionNamedRerouteUsage_19, output=""}
   ]
   properties {subgraphName=ClearCoat}
  }
  MaterialExpressionComposite_8 {
   class=MaterialExpressionComposite
   inputs [RGB {from=MaterialExpressionReroute_44, output=SkinMask}]
   properties {desc=Legacy - Rocket League Feature, subgraphName=ProcessSharpMasks}
  }
  MaterialExpressionComposite_9 {
   class=MaterialExpressionComposite
   inputs [
    SkinMaskTexture {from=MaterialExpressionNamedRerouteUsage_39, output=""}
    BodyTextures {from=MaterialExpressionNamedRerouteUsage_38, output=""}
    EmissiveColor {from=MaterialExpressionNamedRerouteUsage_33, output=""}
   ]
   properties {subgraphName=EmissiveColor}
  }
  MaterialExpressionConstant3Vector_1 {class=MaterialExpressionConstant3Vector, properties {constant=<rgba 0 0 1 0>}}
  MaterialExpressionConstant4Vector_1 {class=MaterialExpressionConstant4Vector, inputs [], properties {desc="Scale(RG)Speed(B)", constant=<rgba 1 1 1 1>}}
  MaterialExpressionConstant4Vector_2 {class=MaterialExpressionConstant4Vector, inputs [], properties {constant=<rgba 1 1 0 -0.05>}}
  MaterialExpressionConstant4Vector_3 {class=MaterialExpressionConstant4Vector, inputs [], properties {constant=<rgba 1 1 0 0.05>}}
  MaterialExpressionConstantBiasScale_0 {
   class=MaterialExpressionConstantBiasScale
   inputs [None {from=MaterialExpressionTextureSample_2, output=A}]
   properties {bias=-0.5, scale=10}
  }
  MaterialExpressionConstantBiasScale_1 {class=MaterialExpressionConstantBiasScale, inputs [None {from=MaterialExpressionPower_0}], properties {bias=-0.5, scale=4}}
  MaterialExpressionConstantBiasScale_10 {class=MaterialExpressionConstantBiasScale, inputs [None {from=MaterialExpressionClamp_12}], properties {bias=0, scale=10}}
  MaterialExpressionConstantBiasScale_12 {class=MaterialExpressionConstantBiasScale, inputs [None {from=MaterialExpressionStaticSwitchParameter_8}], properties {bias=-0.1, scale=2}}
  MaterialExpressionConstantBiasScale_3 {class=MaterialExpressionConstantBiasScale, inputs [None {from=MaterialExpressionMultiply_49}], properties {bias=-0.2, scale=2}}
  MaterialExpressionConstantBiasScale_5 {class=MaterialExpressionConstantBiasScale, inputs [None {from=MaterialExpressionAdd_9}], properties {bias=0, scale=10}}
  MaterialExpressionConstantBiasScale_7 {class=MaterialExpressionConstantBiasScale, inputs [None {from=MaterialExpressionComponentMask_6}], properties {bias=0, scale=2}}
  MaterialExpressionConstantBiasScale_8 {
   class=MaterialExpressionConstantBiasScale
   inputs [None {from=MaterialExpressionTextureSampleParameter2D_14, output=G}]
   properties {bias=0, scale=1}
  }
  MaterialExpressionConstantBiasScale_9 {class=MaterialExpressionConstantBiasScale, inputs [None {from=MaterialExpressionClamp_8}], properties {bias=0, scale=5}}
  MaterialExpressionConstant_0 {class=MaterialExpressionConstant, properties {r=-0.5}}
  MaterialExpressionConstant_1 {class=MaterialExpressionConstant}
  MaterialExpressionConstant_11 {class=MaterialExpressionConstant, properties {r=5}}
  MaterialExpressionConstant_14 {class=MaterialExpressionConstant, properties {r=3}}
  MaterialExpressionConstant_17 {class=MaterialExpressionConstant}
  MaterialExpressionConstant_18 {class=MaterialExpressionConstant, properties {r=-0.15}}
  MaterialExpressionConstant_19 {class=MaterialExpressionConstant, properties {r=1}}
  MaterialExpressionConstant_20 {class=MaterialExpressionConstant, properties {r=-0.05}}
  MaterialExpressionConstant_21 {class=MaterialExpressionConstant, properties {r=0.25}}
  MaterialExpressionConstant_23 {class=MaterialExpressionConstant, properties {r=100}}
  MaterialExpressionConstant_24 {class=MaterialExpressionConstant, properties {r=1}}
  MaterialExpressionConstant_25 {class=MaterialExpressionConstant, properties {r=2}}
  MaterialExpressionConstant_26 {class=MaterialExpressionConstant, properties {r=100}}
  MaterialExpressionConstant_27 {class=MaterialExpressionConstant, properties {r=0.01}}
  MaterialExpressionConstant_28 {class=MaterialExpressionConstant}
  MaterialExpressionConstant_29 {class=MaterialExpressionConstant, properties {r=100}}
  MaterialExpressionConstant_3 {class=MaterialExpressionConstant, properties {r=-0.1}}
  MaterialExpressionConstant_30 {class=MaterialExpressionConstant, properties {r=1}}
  MaterialExpressionConstant_31 {class=MaterialExpressionConstant, properties {r=2}}
  MaterialExpressionConstant_32 {class=MaterialExpressionConstant, properties {r=100}}
  MaterialExpressionConstant_33 {class=MaterialExpressionConstant, properties {r=100}}
  MaterialExpressionConstant_34 {class=MaterialExpressionConstant, properties {r=1}}
  MaterialExpressionConstant_35 {class=MaterialExpressionConstant, properties {r=2}}
  MaterialExpressionConstant_36 {class=MaterialExpressionConstant, properties {r=100}}
  MaterialExpressionConstant_37 {class=MaterialExpressionConstant, properties {r=1}}
  MaterialExpressionConstant_38 {class=MaterialExpressionConstant, properties {r=2}}
  MaterialExpressionConstant_39 {class=MaterialExpressionConstant, properties {r=100}}
  MaterialExpressionConstant_4 {class=MaterialExpressionConstant, properties {r=-0.25}}
  MaterialExpressionConstant_40 {class=MaterialExpressionConstant, properties {r=0.25}}
  MaterialExpressionConstant_6 {class=MaterialExpressionConstant, properties {r=0.5}}
  MaterialExpressionDesaturation_0 {
   class=MaterialExpressionDesaturation
   inputs [
    None {from=MaterialExpressionVectorParameter_16, output=RGB}
    Fraction {from=MaterialExpressionConstant_18, output=""}
   ]
   properties {}
  }
  MaterialExpressionDivide_0 {
   class=MaterialExpressionDivide
   inputs [
    A {from=MaterialExpressionAppendVector_8}
    B {from=MaterialExpressionConstant_23}
   ]
  }
  MaterialExpressionDivide_1 {
   class=MaterialExpressionDivide
   inputs [
    A {from=MaterialExpressionSubtract_12}
    B {from=MaterialExpressionConstant_25}
   ]
  }
  MaterialExpressionDivide_10 {
   class=MaterialExpressionDivide
   inputs [
    A {from=MaterialExpressionSubtract_21}
    B {from=MaterialExpressionConstant_35}
   ]
  }
  MaterialExpressionDivide_11 {
   class=MaterialExpressionDivide
   inputs [
    A {from=MaterialExpressionAppendVector_14}
    B {from=MaterialExpressionConstant_36}
   ]
  }
  MaterialExpressionDivide_12 {
   class=MaterialExpressionDivide
   inputs [
    A {from=MaterialExpressionSubtract_23}
    B {from=MaterialExpressionConstant_38}
   ]
  }
  MaterialExpressionDivide_13 {
   class=MaterialExpressionDivide
   inputs [
    A {from=MaterialExpressionSubtract_24}
    B {from=MaterialExpressionConstant_38}
   ]
  }
  MaterialExpressionDivide_14 {
   class=MaterialExpressionDivide
   inputs [
    A {from=MaterialExpressionScalarParameter_28}
    B {from=MaterialExpressionConstant_39}
   ]
  }
  MaterialExpressionDivide_15 {
   class=MaterialExpressionDivide
   inputs [
    A {from=MaterialExpressionScalarParameter_30}
    B {from=MaterialExpressionConstant_39}
   ]
  }
  MaterialExpressionDivide_2 {
   class=MaterialExpressionDivide
   inputs [
    A {from=MaterialExpressionSubtract_13}
    B {from=MaterialExpressionConstant_25}
   ]
  }
  MaterialExpressionDivide_3 {
   class=MaterialExpressionDivide
   inputs [
    A {from=MaterialExpressionScalarParameter_24}
    B {from=MaterialExpressionConstant_26}
   ]
  }
  MaterialExpressionDivide_4 {
   class=MaterialExpressionDivide
   inputs [
    A {from=MaterialExpressionAppendVector_10}
    B {from=MaterialExpressionConstant_29}
   ]
  }
  MaterialExpressionDivide_5 {
   class=MaterialExpressionDivide
   inputs [
    A {from=MaterialExpressionSubtract_15}
    B {from=MaterialExpressionConstant_31}
   ]
  }
  MaterialExpressionDivide_6 {
   class=MaterialExpressionDivide
   inputs [
    A {from=MaterialExpressionSubtract_16}
    B {from=MaterialExpressionConstant_31}
   ]
  }
  MaterialExpressionDivide_7 {
   class=MaterialExpressionDivide
   inputs [
    A {from=MaterialExpressionScalarParameter_26}
    B {from=MaterialExpressionConstant_32}
   ]
  }
  MaterialExpressionDivide_8 {
   class=MaterialExpressionDivide
   inputs [
    A {from=MaterialExpressionStaticSwitchParameter_13}
    B {from=MaterialExpressionConstant_33}
   ]
  }
  MaterialExpressionDivide_9 {
   class=MaterialExpressionDivide
   inputs [
    A {from=MaterialExpressionSubtract_20}
    B {from=MaterialExpressionConstant_35}
   ]
  }
  MaterialExpressionFresnel_0 {
   class=MaterialExpressionFresnel
   inputs [
    ExponentIn {from="", output=""}
    BaseReflectFractionIn {from="", output=""}
    Normal {from="", output=""}
   ]
   properties {exponent=3}
  }
  MaterialExpressionFunctionOutput_0 {class=MaterialExpressionFunctionOutput, inputs [None {from=MaterialExpressionSetMaterialAttributes_0, output=""}]}
  MaterialExpressionLinearInterpolate_0 {
   class=MaterialExpressionLinearInterpolate
   inputs [
    A {from=MaterialExpressionScalarParameter_2}
    B {from=MaterialExpressionScalarParameter_3}
    Alpha {from=MaterialExpressionMaterialFunctionCall_0, output=G}
   ]
  }
  MaterialExpressionLinearInterpolate_1 {
   class=MaterialExpressionLinearInterpolate
   inputs [
    A {from=MaterialExpressionLinearInterpolate_0}
    B {from=MaterialExpressionScalarParameter_4}
    Alpha {from=MaterialExpressionMaterialFunctionCall_0, output=B}
   ]
  }
  MaterialExpressionLinearInterpolate_10 {
   class=MaterialExpressionLinearInterpolate
   inputs [
    A {from=MaterialExpressionReroute_29, output=BaseColorTexture}
    B {from=MaterialExpressionReroute_27, output=BaseColor}
    Alpha {from=MaterialExpressionComponentMask_0}
   ]
   properties {desc=Put diffuse where R skin mask is black}
  }
  MaterialExpressionLinearInterpolate_11 {
   class=MaterialExpressionLinearInterpolate
   inputs [
    A {from=MaterialExpressionScalarParameter_8}
    B {from=MaterialExpressionScalarParameter_9}
    Alpha {from=MaterialExpressionMaterialFunctionCall_8, output=G}
   ]
  }
  MaterialExpressionLinearInterpolate_12 {
   class=MaterialExpressionLinearInterpolate
   inputs [
    A {from=MaterialExpressionLinearInterpolate_11}
    B {from=MaterialExpressionScalarParameter_10}
    Alpha {from=MaterialExpressionMaterialFunctionCall_8, output=B}
   ]
  }
  MaterialExpressionLinearInterpolate_13 {
   class=MaterialExpressionLinearInterpolate
   inputs [
    A {from=MaterialExpressionReroute_35, output=BottomNormalBaseTexture}
    B {from=MaterialExpressionReroute_36, output=BottomNormalAccentTexture}
    Alpha {from=MaterialExpressionMaterialFunctionCall_3, output=G}
   ]
  }
  MaterialExpressionLinearInterpolate_14 {
   class=MaterialExpressionLinearInterpolate
   inputs [
    A {from=MaterialExpressionLinearInterpolate_13}
    B {from=MaterialExpressionReroute_36, output=BottomNormalAccentTexture}
    Alpha {from=MaterialExpressionMaterialFunctionCall_3, output=B}
   ]
  }
  MaterialExpressionLinearInterpolate_15 {
   class=MaterialExpressionLinearInterpolate
   inputs [
    A {from=MaterialExpressionReroute_31, output=DetailBaseTexture}
    B {from=MaterialExpressionReroute_34, output=DetailAccentTexture}
    Alpha {from=MaterialExpressionSaturate_1}
   ]
   properties {constA=1}
  }
  MaterialExpressionLinearInterpolate_17 {
   class=MaterialExpressionLinearInterpolate
   inputs [
    A {from=MaterialExpressionScalarParameter_11}
    B {from=MaterialExpressionScalarParameter_12}
    Alpha {from=MaterialExpressionMaterialFunctionCall_19, output=G}
   ]
  }
  MaterialExpressionLinearInterpolate_18 {
   class=MaterialExpressionLinearInterpolate
   inputs [
    A {from=MaterialExpressionLinearInterpolate_17}
    B {from=MaterialExpressionScalarParameter_13}
    Alpha {from=MaterialExpressionMaterialFunctionCall_19, output=B}
   ]
  }
  MaterialExpressionLinearInterpolate_2 {
   class=MaterialExpressionLinearInterpolate
   inputs [
    A {from=MaterialExpressionScalarParameter_5}
    B {from=MaterialExpressionScalarParameter_6}
    Alpha {from=MaterialExpressionMaterialFunctionCall_2, output=G}
   ]
  }
  MaterialExpressionLinearInterpolate_3 {
   class=MaterialExpressionLinearInterpolate
   inputs [
    A {from=MaterialExpressionLinearInterpolate_2}
    B {from=MaterialExpressionScalarParameter_7}
    Alpha {from=MaterialExpressionMaterialFunctionCall_2, output=B}
   ]
  }
  MaterialExpressionLinearInterpolate_32 {
   class=MaterialExpressionLinearInterpolate
   inputs [
    A {from=MaterialExpressionConstantBiasScale_3}
    B {from=MaterialExpressionConstant_17}
    Alpha {from=MaterialExpressionClamp_7}
   ]
  }
  MaterialExpressionLinearInterpolate_33 {
   class=MaterialExpressionLinearInterpolate
   inputs [
    A {from=MaterialExpressionConstant_19}
    B {from=MaterialExpressionOneMinus_0}
    Alpha {from=MaterialExpressionScalarParameter_46}
   ]
  }
  MaterialExpressionLinearInterpolate_34 {
   class=MaterialExpressionLinearInterpolate
   inputs [
    A {from=MaterialExpressionVectorParameter_6, output=RGB}
    B {from=MaterialExpressionVectorParameter_16, output=RGB}
    Alpha {from=MaterialExpressionMax_3}
   ]
  }
  MaterialExpressionLinearInterpolate_35 {
   class=MaterialExpressionLinearInterpolate
   inputs [
    A {from=MaterialExpressionMultiply_65}
    B {from=MaterialExpressionVectorParameter_4, output=RGB}
    Alpha {from=MaterialExpressionClamp_17}
   ]
  }
  MaterialExpressionLinearInterpolate_36 {
   class=MaterialExpressionLinearInterpolate
   inputs [
    A {from=MaterialExpressionMultiply_70}
    B {from=MaterialExpressionMultiply_71}
    Alpha {from=MaterialExpressionScalarParameter_18}
   ]
  }
  MaterialExpressionLinearInterpolate_37 {
   class=MaterialExpressionLinearInterpolate
   inputs [
    A {from=MaterialExpressionMultiply_74}
    B {from=MaterialExpressionMultiply_64}
    Alpha {from=MaterialExpressionClamp_16}
   ]
  }
  MaterialExpressionLinearInterpolate_38 {
   class=MaterialExpressionLinearInterpolate
   inputs [
    A {from=MaterialExpressionMultiply_76}
    B {from=MaterialExpressionMultiply_78}
    Alpha {from=MaterialExpressionClamp_16}
   ]
  }
  MaterialExpressionLinearInterpolate_39 {
   class=MaterialExpressionLinearInterpolate
   inputs [
    A {from=MaterialExpressionAdd_26}
    B {from=MaterialExpressionMultiply_88}
    Alpha {from=MaterialExpressionClamp_18}
   ]
  }
  MaterialExpressionLinearInterpolate_4 {
   class=MaterialExpressionLinearInterpolate
   inputs [
    A {from=MaterialExpressionConstant3Vector_1}
    B {from=MaterialExpressionTextureSampleParameter2D_5, output=RGB}
    Alpha {from=MaterialExpressionReroute_21, output=LOD}
   ]
   properties {constAlpha=0}
  }
  MaterialExpressionLinearInterpolate_40 {
   class=MaterialExpressionLinearInterpolate
   inputs [
    A {from=MaterialExpressionAdd_22}
    B {from=MaterialExpressionMultiply_89}
    Alpha {from=MaterialExpressionClamp_18}
   ]
  }
  MaterialExpressionLinearInterpolate_41 {
   class=MaterialExpressionLinearInterpolate
   inputs [
    A {from=MaterialExpressionLinearInterpolate_40}
    B {from=MaterialExpressionMultiply_81}
    Alpha {from=MaterialExpressionStaticSwitchParameter_12}
   ]
  }
  MaterialExpressionLinearInterpolate_42 {
   class=MaterialExpressionLinearInterpolate
   inputs [
    A {from=MaterialExpressionLinearInterpolate_39}
    B {from=MaterialExpressionLinearInterpolate_43}
    Alpha {from=MaterialExpressionStaticSwitchParameter_12}
   ]
  }
  MaterialExpressionLinearInterpolate_43 {
   class=MaterialExpressionLinearInterpolate
   inputs [
    A {from=MaterialExpressionVectorParameter_4, output=RGB}
    B {from=MaterialExpressionVectorParameter_6, output=RGB}
    Alpha {from=MaterialExpressionMax_7}
   ]
  }
  MaterialExpressionLinearInterpolate_44 {
   class=MaterialExpressionLinearInterpolate
   inputs [
    A {from=MaterialExpressionTextureSampleParameter2D_18, output=RGB}
    B {from=MaterialExpressionTextureSampleParameter2D_19, output=RGB}
    Alpha {from=MaterialExpressionTextureSampleParameter2D_19, output=RGB}
   ]
  }
  MaterialExpressionLinearInterpolate_5 {
   class=MaterialExpressionLinearInterpolate
   inputs [
    A {from=MaterialExpressionMaterialFunctionCall_5, output=R}
    B {from=MaterialExpressionMaterialFunctionCall_5, output=R}
    Alpha {from=MaterialExpressionMaterialFunctionCall_4, output=G}
   ]
  }
  MaterialExpressionLinearInterpolate_6 {
   class=MaterialExpressionLinearInterpolate
   inputs [
    A {from=MaterialExpressionLinearInterpolate_5}
    B {from=MaterialExpressionMaterialFunctionCall_5, output=B}
    Alpha {from=MaterialExpressionMaterialFunctionCall_4, output=B}
   ]
  }
  MaterialExpressionLinearInterpolate_7 {
   class=MaterialExpressionLinearInterpolate
   inputs [
    A {from=MaterialExpressionReroute_8, output=TeamColor}
    B {from=MaterialExpressionReroute_13, output=AccentColor}
    Alpha {from=MaterialExpressionMaterialFunctionCall_6, output=G}
   ]
  }
  MaterialExpressionLinearInterpolate_8 {
   class=MaterialExpressionLinearInterpolate
   inputs [
    A {from=MaterialExpressionLinearInterpolate_7}
    B {from=MaterialExpressionReroute_18, output=PaintColor}
    Alpha {from=MaterialExpressionMaterialFunctionCall_6, output=B}
   ]
  }
  MaterialExpressionMaterialFunctionCall_0 {
   class=MaterialExpressionMaterialFunctionCall
   inputs [Color {from=MaterialExpressionReroute_10, output=SkinMaskTexture}]
   properties {materialFunction=&SplitComponents}
  }
  MaterialExpressionMaterialFunctionCall_1 {
   class=MaterialExpressionMaterialFunctionCall
   inputs [
    Normal {from=MaterialExpressionTextureSampleParameter2D_6, output=RGB}
    Flatness {from=MaterialExpressionReroute_40}
   ]
   properties {materialFunction=&FlattenNormal}
  }
  MaterialExpressionMaterialFunctionCall_10 {
   class=MaterialExpressionMaterialFunctionCall
   inputs [Color {from=MaterialExpressionReroute_16, output=SkinMaskTexture}]
   properties {materialFunction=&SplitComponents}
  }
  MaterialExpressionMaterialFunctionCall_19 {
   class=MaterialExpressionMaterialFunctionCall
   inputs [Color {from=MaterialExpressionReroute_11, output=SkinMaskTexture}]
   properties {materialFunction=&SplitComponents}
  }
  MaterialExpressionMaterialFunctionCall_2 {
   class=MaterialExpressionMaterialFunctionCall
   inputs [Color {from=MaterialExpressionReroute_10, output=SkinMaskTexture}]
   properties {materialFunction=&SplitComponents}
  }
  MaterialExpressionMaterialFunctionCall_3 {
   class=MaterialExpressionMaterialFunctionCall
   inputs [Color {from=MaterialExpressionReroute_11, output=SkinMaskTexture}]
   properties {materialFunction=&SplitComponents}
  }
  MaterialExpressionMaterialFunctionCall_4 {
   class=MaterialExpressionMaterialFunctionCall
   inputs [Color {from=MaterialExpressionReroute_3, output=SkinMask}]
   properties {materialFunction=&SplitComponents}
  }
  MaterialExpressionMaterialFunctionCall_5 {
   class=MaterialExpressionMaterialFunctionCall
   inputs [Color {from=MaterialExpressionReroute_2, output=Roughness}]
   properties {materialFunction=&SplitComponents}
  }
  MaterialExpressionMaterialFunctionCall_6 {
   class=MaterialExpressionMaterialFunctionCall
   inputs [Color {from=MaterialExpressionReroute_20, output=SkinMaskTexture}]
   properties {materialFunction=&SplitComponents}
  }
  MaterialExpressionMaterialFunctionCall_7 {
   class=MaterialExpressionMaterialFunctionCall
   inputs [
    BaseNormal {from=MaterialExpressionNamedRerouteUsage_31}
    AdditionalNormal {from=MaterialExpressionLinearInterpolate_14}
   ]
   properties {materialFunction=&BlendAngleCorrectedNormals}
  }
  MaterialExpressionMaterialFunctionCall_8 {
   class=MaterialExpressionMaterialFunctionCall
   inputs [Color {from=MaterialExpressionReroute_11, output=SkinMaskTexture}]
   properties {materialFunction=&SplitComponents}
  }
  MaterialExpressionMaterialFunctionCall_9 {
   class=MaterialExpressionMaterialFunctionCall
   inputs [
    Normal {from=MaterialExpressionTextureSampleParameter2D_7, output=RGB}
    Flatness {from=MaterialExpressionReroute_40}
   ]
   properties {materialFunction=&FlattenNormal}
  }
  MaterialExpressionMax_1 {
   class=MaterialExpressionMax
   inputs [
    A {from=MaterialExpressionTextureSampleParameter2D_14, output=G}
    B {from=MaterialExpressionTextureSampleParameter2D_15, output=B}
   ]
  }
  MaterialExpressionMax_2 {
   class=MaterialExpressionMax
   inputs [
    A {from=MaterialExpressionClamp_7}
    B {from=MaterialExpressionClamp_9}
   ]
  }
  MaterialExpressionMax_3 {
   class=MaterialExpressionMax
   inputs [
    A {from=MaterialExpressionClamp_10}
    B {from=MaterialExpressionClamp_9}
   ]
  }
  MaterialExpressionMax_7 {
   class=MaterialExpressionMax
   inputs [
    A {from=MaterialExpressionTextureSampleParameter2D_17, output=G}
    B {from=MaterialExpressionTextureSampleParameter2D_16, output=G}
   ]
  }
  MaterialExpressionMax_8 {
   class=MaterialExpressionMax
   inputs [
    A {from=MaterialExpressionTextureSampleParameter2D_16, output=R}
    B {from=MaterialExpressionTextureSampleParameter2D_17, output=R}
   ]
  }
  MaterialExpressionMax_9 {
   class=MaterialExpressionMax
   inputs [
    A {from=MaterialExpressionTextureSampleParameter2D_18, output=A}
    B {from=MaterialExpressionTextureSampleParameter2D_19, output=A}
   ]
  }
  MaterialExpressionMultiply_0 {
   class=MaterialExpressionMultiply
   inputs [
    A {from=MaterialExpressionReroute_49, output=Result}
    B {from=MaterialExpressionNamedRerouteUsage_1}
   ]
  }
  MaterialExpressionMultiply_1 {
   class=MaterialExpressionMultiply
   inputs [
    A {from=MaterialExpressionReroute_47, output=BaseColorInput}
    B {from=MaterialExpressionSaturate_2}
   ]
  }
  MaterialExpressionMultiply_12 {
   class=MaterialExpressionMultiply
   inputs [
    A {from=MaterialExpressionTime_6}
    B {from=MaterialExpressionScalarParameter_44}
   ]
  }
  MaterialExpressionMultiply_14 {
   class=MaterialExpressionMultiply
   inputs [
    A {from=MaterialExpressionClamp_14}
    B {from=MaterialExpressionConstant_11}
   ]
  }
  MaterialExpressionMultiply_2 {
   class=MaterialExpressionMultiply
   inputs [
    A {from=MaterialExpressionMultiply_0}
    B {from=MaterialExpressionNamedRerouteUsage_24}
   ]
  }
  MaterialExpressionMultiply_3 {
   class=MaterialExpressionMultiply
   inputs [
    A {from=MaterialExpressionTextureCoordinate_0}
    B {from=MaterialExpressionScalarParameter_15}
   ]
  }
  MaterialExpressionMultiply_36 {
   class=MaterialExpressionMultiply
   inputs [
    A {from=MaterialExpressionTextureCoordinate_12}
    B {from=MaterialExpressionAppendVector_17}
   ]
  }
  MaterialExpressionMultiply_37 {
   class=MaterialExpressionMultiply
   inputs [
    A {from=MaterialExpressionTextureSample_0, output=B}
    B {from=MaterialExpressionScalarParameter_41}
   ]
  }
  MaterialExpressionMultiply_4 {
   class=MaterialExpressionMultiply
   inputs [
    A {from=MaterialExpressionMultiply_6}
    B {from=MaterialExpressionReroute_60}
   ]
  }
  MaterialExpressionMultiply_44 {
   class=MaterialExpressionMultiply
   inputs [
    A {from=MaterialExpressionTextureSample_0, output=B}
    B {from=MaterialExpressionScalarParameter_42}
   ]
  }
  MaterialExpressionMultiply_45 {
   class=MaterialExpressionMultiply
   inputs [
    A {from=MaterialExpressionTime_4}
    B {from=MaterialExpressionScalarParameter_43}
   ]
  }
  MaterialExpressionMultiply_49 {
   class=MaterialExpressionMultiply
   inputs [
    A {from=MaterialExpressionConstantBiasScale_0}
    B {from=MaterialExpressionTextureSample_3, output=R}
   ]
  }
  MaterialExpressionMultiply_5 {
   class=MaterialExpressionMultiply
   inputs [
    A {from=MaterialExpressionTextureCoordinate_1}
    B {from=MaterialExpressionScalarParameter_0}
   ]
  }
  MaterialExpressionMultiply_50 {
   class=MaterialExpressionMultiply
   inputs [
    A {from=MaterialExpressionConstant4Vector_1, output=B}
    B {from=MaterialExpressionTime_7}
   ]
  }
  MaterialExpressionMultiply_51 {
   class=MaterialExpressionMultiply
   inputs [
    A {from=MaterialExpressionTextureCoordinate_7}
    B {from=MaterialExpressionAppendVector_2}
   ]
  }
  MaterialExpressionMultiply_52 {
   class=MaterialExpressionMultiply
   inputs [
    A {from=MaterialExpressionTextureCoordinate_6}
    B {from=MaterialExpressionAppendVector_2}
   ]
  }
  MaterialExpressionMultiply_53 {
   class=MaterialExpressionMultiply
   inputs [
    A {from=MaterialExpressionPower_6}
    B {from=MaterialExpressionLinearInterpolate_32}
   ]
  }
  MaterialExpressionMultiply_54 {
   class=MaterialExpressionMultiply
   inputs [
    A {from=MaterialExpressionClamp_3}
    B {from=MaterialExpressionScalarParameter_45}
   ]
  }
  MaterialExpressionMultiply_55 {
   class=MaterialExpressionMultiply
   inputs [
    A {from=MaterialExpressionClamp_4}
    B {from=MaterialExpressionMultiply_54}
   ]
  }
  MaterialExpressionMultiply_56 {
   class=MaterialExpressionMultiply
   inputs [
    A {from=MaterialExpressionTextureSampleParameter2D_15, output=B}
    B {from=MaterialExpressionComponentMask_29}
   ]
  }
  MaterialExpressionMultiply_57 {
   class=MaterialExpressionMultiply
   inputs [
    A {from=MaterialExpressionTextureSampleParameter2D_15, output=B}
    B {from=MaterialExpressionComponentMask_30}
   ]
  }
  MaterialExpressionMultiply_58 {
   class=MaterialExpressionMultiply
   inputs [
    A {from=MaterialExpressionAdd_15}
    B {from=MaterialExpressionLinearInterpolate_33}
   ]
  }
  MaterialExpressionMultiply_59 {
   class=MaterialExpressionMultiply
   inputs [
    A {from=MaterialExpressionTextureCoordinate_13}
    B {from=MaterialExpressionAppendVector_4}
   ]
  }
  MaterialExpressionMultiply_6 {
   class=MaterialExpressionMultiply
   inputs [
    A {from=MaterialExpressionChannelMaskParameter_1, output=R}
    B {from=MaterialExpressionScalarParameter_16}
   ]
  }
  MaterialExpressionMultiply_60 {
   class=MaterialExpressionMultiply
   inputs [
    A {from=MaterialExpressionAppendVector_5}
    B {from=MaterialExpressionMultiply_45}
   ]
  }
  MaterialExpressionMultiply_61 {
   class=MaterialExpressionMultiply
   inputs [
    A {from=MaterialExpressionTextureCoordinate_2}
    B {from=MaterialExpressionAppendVector_6}
   ]
  }
  MaterialExpressionMultiply_62 {
   class=MaterialExpressionMultiply
   inputs [
    A {from=MaterialExpressionMultiply_12}
    B {from=MaterialExpressionAppendVector_7}
   ]
  }
  MaterialExpressionMultiply_63 {
   class=MaterialExpressionMultiply
   inputs [
    A {from=MaterialExpressionLinearInterpolate_34}
    B {from=MaterialExpressionClamp_11}
   ]
  }
  MaterialExpressionMultiply_64 {
   class=MaterialExpressionMultiply
   inputs [
    A {from=MaterialExpressionScalarParameter_47}
    B {from=MaterialExpressionMultiply_63}
   ]
  }
  MaterialExpressionMultiply_65 {
   class=MaterialExpressionMultiply
   inputs [
    A {from=MaterialExpressionVectorParameter_4, output=RGB}
    B {from=MaterialExpressionScalarParameter_48}
   ]
  }
  MaterialExpressionMultiply_66 {
   class=MaterialExpressionMultiply
   inputs [
    A {from=MaterialExpressionPower_8}
    B {from=MaterialExpressionScalarParameter_50}
   ]
  }
  MaterialExpressionMultiply_67 {
   class=MaterialExpressionMultiply
   inputs [
    A {from=MaterialExpressionVectorParameter_4, output=RGB}
    B {from=MaterialExpressionClamp_17}
   ]
  }
  MaterialExpressionMultiply_68 {
   class=MaterialExpressionMultiply
   inputs [
    A {from=MaterialExpressionPower_2}
    B {from=MaterialExpressionScalarParameter_52}
   ]
  }
  MaterialExpressionMultiply_69 {
   class=MaterialExpressionMultiply
   inputs [
    A {from=MaterialExpressionComponentMask_16}
    B {from=MaterialExpressionMultiply_68}
   ]
  }
  MaterialExpressionMultiply_7 {
   class=MaterialExpressionMultiply
   inputs [
    A {from=MaterialExpressionNamedRerouteUsage_6}
    B {from=MaterialExpressionReroute_57, output=Output}
   ]
  }
  MaterialExpressionMultiply_70 {
   class=MaterialExpressionMultiply
   inputs [
    A {from=MaterialExpressionOneMinus_4}
    B {from=MaterialExpressionMultiply_69}
   ]
  }
  MaterialExpressionMultiply_71 {
   class=MaterialExpressionMultiply
   inputs [
    A {from=MaterialExpressionComponentMask_16}
    B {from=MaterialExpressionMultiply_69}
   ]
  }
  MaterialExpressionMultiply_72 {
   class=MaterialExpressionMultiply
   inputs [
    A {from=MaterialExpressionVectorParameter_4, output=RGB}
    B {from=MaterialExpressionScalarParameter_53}
   ]
  }
  MaterialExpressionMultiply_73 {
   class=MaterialExpressionMultiply
   inputs [
    A {from=MaterialExpressionMultiply_72}
    B {from=MaterialExpressionLinearInterpolate_36}
   ]
  }
  MaterialExpressionMultiply_74 {
   class=MaterialExpressionMultiply
   inputs [
    A {from=MaterialExpressionMultiply_67}
    B {from=MaterialExpressionScalarParameter_19}
   ]
  }
  MaterialExpressionMultiply_75 {
   class=MaterialExpressionMultiply
   inputs [
    A {from=MaterialExpressionMultiply_64}
    B {from=MaterialExpressionConstant_21}
   ]
  }
  MaterialExpressionMultiply_76 {
   class=MaterialExpressionMultiply
   inputs [
    A {from=MaterialExpressionAdd_13}
    B {from=MaterialExpressionScalarParameter_22}
   ]
  }
  MaterialExpressionMultiply_77 {
   class=MaterialExpressionMultiply
   inputs [
    A {from=MaterialExpressionAdd_23}
    B {from=MaterialExpressionScalarParameter_20}
   ]
  }
  MaterialExpressionMultiply_78 {
   class=MaterialExpressionMultiply
   inputs [
    A {from=MaterialExpressionMultiply_76}
    B {from=MaterialExpressionScalarParameter_23}
   ]
  }
  MaterialExpressionMultiply_79 {
   class=MaterialExpressionMultiply
   inputs [
    A {from=MaterialExpressionComponentMask_17}
    B {from=MaterialExpressionVectorParameter_7, output=R}
   ]
  }
  MaterialExpressionMultiply_8 {
   class=MaterialExpressionMultiply
   inputs [
    A {from=MaterialExpressionSubtract_1}
    B {from=""}
   ]
   properties {constB=1.1087}
  }
  MaterialExpressionMultiply_80 {
   class=MaterialExpressionMultiply
   inputs [
    A {from=MaterialExpressionComponentMask_18}
    B {from=MaterialExpressionVectorParameter_7, output=G}
   ]
  }
  MaterialExpressionMultiply_81 {
   class=MaterialExpressionMultiply
   inputs [
    A {from=MaterialExpressionConstant_27}
    B {from=MaterialExpressionLinearInterpolate_43}
   ]
  }
  MaterialExpressionMultiply_82 {
   class=MaterialExpressionMultiply
   inputs [
    A {from=MaterialExpressionComponentMask_20}
    B {from=MaterialExpressionVectorParameter_9, output=R}
   ]
  }
  MaterialExpressionMultiply_83 {
   class=MaterialExpressionMultiply
   inputs [
    A {from=MaterialExpressionComponentMask_21}
    B {from=MaterialExpressionVectorParameter_9, output=G}
   ]
  }
  MaterialExpressionMultiply_84 {
   class=MaterialExpressionMultiply
   inputs [
    A {from=MaterialExpressionComponentMask_22}
    B {from=MaterialExpressionVectorParameter_13, output=R}
   ]
  }
  MaterialExpressionMultiply_85 {
   class=MaterialExpressionMultiply
   inputs [
    A {from=MaterialExpressionComponentMask_23}
    B {from=MaterialExpressionVectorParameter_13, output=G}
   ]
  }
  MaterialExpressionMultiply_86 {
   class=MaterialExpressionMultiply
   inputs [
    A {from=MaterialExpressionComponentMask_24}
    B {from=MaterialExpressionVectorParameter_13, output=R}
   ]
  }
  MaterialExpressionMultiply_87 {
   class=MaterialExpressionMultiply
   inputs [
    A {from=MaterialExpressionComponentMask_25}
    B {from=MaterialExpressionVectorParameter_13, output=G}
   ]
  }
  MaterialExpressionMultiply_88 {
   class=MaterialExpressionMultiply
   inputs [
    A {from=MaterialExpressionScalarParameter_29}
    B {from=MaterialExpressionStaticSwitchParameter_15}
   ]
  }
  MaterialExpressionMultiply_89 {
   class=MaterialExpressionMultiply
   inputs [
    A {from=MaterialExpressionMultiply_88}
    B {from=MaterialExpressionConstant_40}
   ]
  }
  MaterialExpressionNamedRerouteDeclaration_0 {
   class=MaterialExpressionNamedRerouteDeclaration
   inputs [None {from=MaterialExpressionNamedRerouteUsage_2}]
   properties {name=MetallicMasked, nodeColor=<rgba 0 1 0.7656 1>}
  }
  MaterialExpressionNamedRerouteDeclaration_1 {
   class=MaterialExpressionNamedRerouteDeclaration
   inputs [None {from=MaterialExpressionStaticSwitchParameter_3}]
   properties {name=UV, nodeColor=<rgba 1 0.8646 0.8646 1>}
  }
  MaterialExpressionNamedRerouteDeclaration_10 {
   class=MaterialExpressionNamedRerouteDeclaration
   inputs [None {from=MaterialExpressionReroute_45, output=SharpMasksSwitchOutput}]
   properties {name=SkinMaskTexture, nodeColor=<rgba 0.6038 0.3325 0.013 1>}
  }
  MaterialExpressionNamedRerouteDeclaration_11 {
   class=MaterialExpressionNamedRerouteDeclaration
   inputs [None {from=MaterialExpressionReroute_5, output=RoughnessMasked}]
   properties {name=RoughnessMasked, nodeColor=<rgba 0 1 0.7656 1>}
  }
  MaterialExpressionNamedRerouteDeclaration_12 {
   class=MaterialExpressionNamedRerouteDeclaration
   inputs [None {from=MaterialExpressionReroute_23, output=TeamColor}]
   properties {name=TeamColor, nodeColor=<rgba 0.9219 0 1 1>}
  }
  MaterialExpressionNamedRerouteDeclaration_13 {
   class=MaterialExpressionNamedRerouteDeclaration
   inputs [None {from=MaterialExpressionReroute_24, output=AccentColor}]
   properties {name=AccentColor, nodeColor=<rgba 0.9219 0 1 1>}
  }
  MaterialExpressionNamedRerouteDeclaration_14 {
   class=MaterialExpressionNamedRerouteDeclaration
   inputs [None {from=MaterialExpressionReroute_25, output=PaintColor}]
   properties {name=PaintColor, nodeColor=<rgba 0.9219 0 1 1>}
  }
  MaterialExpressionNamedRerouteDeclaration_15 {
   class=MaterialExpressionNamedRerouteDeclaration
   inputs [None {from=MaterialExpressionReroute_15, output=BottomNormalBase}]
   properties {name=BottomNormalBaseTexture, nodeColor=<rgba 0.6038 0.3325 0.013 1>}
  }
  MaterialExpressionNamedRerouteDeclaration_16 {
   class=MaterialExpressionNamedRerouteDeclaration
   inputs [None {from=MaterialExpressionReroute_7, output=BaseColor}]
   properties {name=BaseColorTexture, nodeColor=<rgba 0.6038 0.3325 0.013 1>}
  }
  MaterialExpressionNamedRerouteDeclaration_17 {
   class=MaterialExpressionNamedRerouteDeclaration
   inputs [None {from=MaterialExpressionLinearInterpolate_41}]
   properties {name=BaseColor, nodeColor=<rgba 0 1 0.7656 1>}
  }
  MaterialExpressionNamedRerouteDeclaration_18 {
   class=MaterialExpressionNamedRerouteDeclaration
   inputs [None {from=MaterialExpressionNamedRerouteUsage_20}]
   properties {name=CCThicknessMasked, nodeColor=<rgba 0 1 0.7656 1>}
  }
  MaterialExpressionNamedRerouteDeclaration_19 {
   class=MaterialExpressionNamedRerouteDeclaration
   inputs [None {from=MaterialExpressionNamedRerouteUsage_17}]
   properties {name=CCRoughnessMasked, nodeColor=<rgba 0 1 0.7656 1>}
  }
  MaterialExpressionNamedRerouteDeclaration_2 {
   class=MaterialExpressionNamedRerouteDeclaration
   inputs [None {from=MaterialExpressionReroute_6, output=Metallic}]
   properties {name=Metallic, nodeColor=<rgba 0 0.1434 1 1>}
  }
  MaterialExpressionNamedRerouteDeclaration_20 {
   class=MaterialExpressionNamedRerouteDeclaration
   inputs [None {from=MaterialExpressionReroute_38, output=BottomNormalAccent}]
   properties {name=BottomNormalAccentTexture, nodeColor=<rgba 0.6038 0.3325 0.013 1>}
  }
  MaterialExpressionNamedRerouteDeclaration_21 {
   class=MaterialExpressionNamedRerouteDeclaration
   inputs [None {from=MaterialExpressionReroute_19, output=AO}]
   properties {name=AOTexture, nodeColor=<rgba 0.6038 0.3325 0.013 1>}
  }
  MaterialExpressionNamedRerouteDeclaration_22 {
   class=MaterialExpressionNamedRerouteDeclaration
   inputs [None {from=MaterialExpressionReroute_14, output=BodyNormalLOD}]
   properties {name=BodyNormalLOD, nodeColor=<rgba 0.6038 0.3325 0.013 1>}
  }
  MaterialExpressionNamedRerouteDeclaration_23 {
   class=MaterialExpressionNamedRerouteDeclaration
   inputs [None {from=MaterialExpressionReroute_37, output=ClearCoatBottomNormal}]
   properties {name=CCBottomNormal, nodeColor=<rgba 0 0.1434 1 1>}
  }
  MaterialExpressionNamedRerouteDeclaration_24 {
   class=MaterialExpressionNamedRerouteDeclaration
   inputs [None {from=MaterialExpressionReroute_53, output=PBR}]
   properties {name=BodyTextures_RoughnessMetallic, nodeColor=<rgba 1 0.3516 0 1>}
  }
  MaterialExpressionNamedRerouteDeclaration_25 {
   class=MaterialExpressionNamedRerouteDeclaration
   inputs [None {from=MaterialExpressionMultiply_3}]
   properties {name=UV_BottomNormal, nodeColor=<rgba 0.6458 0.6458 0.6458 1>}
  }
  MaterialExpressionNamedRerouteDeclaration_26 {
   class=MaterialExpressionNamedRerouteDeclaration
   inputs [None {from=MaterialExpressionReroute_46, output=TextureRawOutput}]
   properties {name=SkinMaskTextureRaw, nodeColor=<rgba 1 0 0.2812 1>}
  }
  MaterialExpressionNamedRerouteDeclaration_27 {
   class=MaterialExpressionNamedRerouteDeclaration
   inputs [None {from=MaterialExpressionAdd_14}]
   properties {name=EmissiveOutput, nodeColor=<rgba 0 0.9297 1 1>}
  }
  MaterialExpressionNamedRerouteDeclaration_28 {
   class=MaterialExpressionNamedRerouteDeclaration
   inputs [None {from=MaterialExpressionStaticSwitchParameter_5}]
   properties {name=MetallicOutput, nodeColor=<rgba 0 1 0.7647 1>}
  }
  MaterialExpressionNamedRerouteDeclaration_29 {
   class=MaterialExpressionNamedRerouteDeclaration
   inputs [None {from=MaterialExpressionStaticSwitchParameter_6}]
   properties {name=RoughnessOutput, nodeColor=<rgba 0 1 0.7647 1>}
  }
  MaterialExpressionNamedRerouteDeclaration_3 {
   class=MaterialExpressionNamedRerouteDeclaration
   inputs [None {from=MaterialExpressionReroute_9, output=Roughness}]
   properties {name=Roughness, nodeColor=<rgba 0 0.1434 1 1>}
  }
  MaterialExpressionNamedRerouteDeclaration_30 {
   class=MaterialExpressionNamedRerouteDeclaration
   inputs [None {from=MaterialExpressionReroute_54, output=EmissiveColor}]
   properties {name=EmissiveColor, nodeColor=<rgba 0.9219 0 1 1>}
  }
  MaterialExpressionNamedRerouteDeclaration_32 {
   class=MaterialExpressionNamedRerouteDeclaration
   inputs [None {from=MaterialExpressionReroute_22, output=Curvature}]
   properties {name=Curvature, nodeColor=<rgba 0.6038 0.3325 0.013 1>}
  }
  MaterialExpressionNamedRerouteDeclaration_4 {
   class=MaterialExpressionNamedRerouteDeclaration
   inputs [None {from=MaterialExpressionReroute_32, output=ClearCoat Thickness}]
   properties {name=CCThickness, nodeColor=<rgba 0 0.1434 1 1>}
  }
  MaterialExpressionNamedRerouteDeclaration_5 {
   class=MaterialExpressionNamedRerouteDeclaration
   inputs [None {from=MaterialExpressionReroute_33, output=ClearCoat Roughness}]
   properties {name=CCRoughness, nodeColor=<rgba 0 0.1434 1 1>}
  }
  MaterialExpressionNamedRerouteDeclaration_6 {
   class=MaterialExpressionNamedRerouteDeclaration
   inputs [None {from=MaterialExpressionReroute_0, output=DetailBase}]
   properties {name=DetailBaseTexture, nodeColor=<rgba 0.6038 0.3325 0.013 1>}
  }
  MaterialExpressionNamedRerouteDeclaration_7 {
   class=MaterialExpressionNamedRerouteDeclaration
   inputs [None {from=MaterialExpressionReroute_1, output=DetailAccent}]
   properties {name=DetailAccentTexture, nodeColor=<rgba 0.6038 0.3325 0.013 1>}
  }
  MaterialExpressionNamedRerouteDeclaration_8 {
   class=MaterialExpressionNamedRerouteDeclaration
   inputs [None {from=MaterialExpressionReroute_17, output=DetailsMasked}]
   properties {name=DetailsMasked, nodeColor=<rgba 0 1 0.7656 1>}
  }
  MaterialExpressionNamedRerouteDeclaration_9 {
   class=MaterialExpressionNamedRerouteDeclaration
   inputs [None {from=MaterialExpressionMultiply_5}]
   properties {name=UV_Detail, nodeColor=<rgba 1 0.8646 0.8646 1>}
  }
  MaterialExpressionNamedRerouteUsage_0 {class=MaterialExpressionNamedRerouteUsage}
  MaterialExpressionNamedRerouteUsage_1 {class=MaterialExpressionNamedRerouteUsage}
  MaterialExpressionNamedRerouteUsage_10 {class=MaterialExpressionNamedRerouteUsage}
  MaterialExpressionNamedRerouteUsage_11 {class=MaterialExpressionNamedRerouteUsage}
  MaterialExpressionNamedRerouteUsage_12 {class=MaterialExpressionNamedRerouteUsage}
  MaterialExpressionNamedRerouteUsage_13 {class=MaterialExpressionNamedRerouteUsage}
  MaterialExpressionNamedRerouteUsage_14 {class=MaterialExpressionNamedRerouteUsage}
  MaterialExpressionNamedRerouteUsage_15 {class=MaterialExpressionNamedRerouteUsage}
  MaterialExpressionNamedRerouteUsage_16 {class=MaterialExpressionNamedRerouteUsage}
  MaterialExpressionNamedRerouteUsage_17 {class=MaterialExpressionNamedRerouteUsage}
  MaterialExpressionNamedRerouteUsage_18 {class=MaterialExpressionNamedRerouteUsage}
  MaterialExpressionNamedRerouteUsage_19 {class=MaterialExpressionNamedRerouteUsage}
  MaterialExpressionNamedRerouteUsage_2 {class=MaterialExpressionNamedRerouteUsage}
  MaterialExpressionNamedRerouteUsage_20 {class=MaterialExpressionNamedRerouteUsage}
  MaterialExpressionNamedRerouteUsage_21 {class=MaterialExpressionNamedRerouteUsage}
  MaterialExpressionNamedRerouteUsage_22 {class=MaterialExpressionNamedRerouteUsage}
  MaterialExpressionNamedRerouteUsage_23 {class=MaterialExpressionNamedRerouteUsage}
  MaterialExpressionNamedRerouteUsage_24 {class=MaterialExpressionNamedRerouteUsage}
  MaterialExpressionNamedRerouteUsage_25 {class=MaterialExpressionNamedRerouteUsage}
  MaterialExpressionNamedRerouteUsage_26 {class=MaterialExpressionNamedRerouteUsage}
  MaterialExpressionNamedRerouteUsage_27 {class=MaterialExpressionNamedRerouteUsage}
  MaterialExpressionNamedRerouteUsage_28 {class=MaterialExpressionNamedRerouteUsage}
  MaterialExpressionNamedRerouteUsage_29 {class=MaterialExpressionNamedRerouteUsage}
  MaterialExpressionNamedRerouteUsage_3 {class=MaterialExpressionNamedRerouteUsage}
  MaterialExpressionNamedRerouteUsage_30 {class=MaterialExpressionNamedRerouteUsage}
  MaterialExpressionNamedRerouteUsage_31 {class=MaterialExpressionNamedRerouteUsage}
  MaterialExpressionNamedRerouteUsage_32 {class=MaterialExpressionNamedRerouteUsage}
  MaterialExpressionNamedRerouteUsage_33 {class=MaterialExpressionNamedRerouteUsage}
  MaterialExpressionNamedRerouteUsage_34 {class=MaterialExpressionNamedRerouteUsage}
  MaterialExpressionNamedRerouteUsage_35 {class=MaterialExpressionNamedRerouteUsage}
  MaterialExpressionNamedRerouteUsage_36 {class=MaterialExpressionNamedRerouteUsage}
  MaterialExpressionNamedRerouteUsage_37 {class=MaterialExpressionNamedRerouteUsage}
  MaterialExpressionNamedRerouteUsage_38 {class=MaterialExpressionNamedRerouteUsage}
  MaterialExpressionNamedRerouteUsage_39 {class=MaterialExpressionNamedRerouteUsage}
  MaterialExpressionNamedRerouteUsage_4 {class=MaterialExpressionNamedRerouteUsage}
  MaterialExpressionNamedRerouteUsage_5 {class=MaterialExpressionNamedRerouteUsage}
  MaterialExpressionNamedRerouteUsage_6 {class=MaterialExpressionNamedRerouteUsage}
  MaterialExpressionNamedRerouteUsage_7 {class=MaterialExpressionNamedRerouteUsage}
  MaterialExpressionOneMinus_0 {class=MaterialExpressionOneMinus, inputs [None {from=MaterialExpressionComponentMask_31}]}
  MaterialExpressionOneMinus_4 {class=MaterialExpressionOneMinus, inputs [None {from=MaterialExpressionComponentMask_16}]}
  MaterialExpressionPanner_7 {
   class=MaterialExpressionPanner
   inputs [
    Coordinate {from=MaterialExpressionMultiply_52}
    Time {from=MaterialExpressionMultiply_50}
    Speed {from=""}
   ]
   properties {speedY=0.02}
  }
  MaterialExpressionPanner_8 {
   class=MaterialExpressionPanner
   inputs [
    Coordinate {from=MaterialExpressionMultiply_51}
    Time {from=MaterialExpressionMultiply_50}
    Speed {from=""}
   ]
   properties {speedX=0.05, speedY=1}
  }
  MaterialExpressionPinBase_0 {
   class=MaterialExpressionPinBase
   inputs []
   properties {reroutePins [SkinMaskTexture {expression=&self:MaterialExpressionReroute_10}]}
  }
  MaterialExpressionPinBase_1 {
   class=MaterialExpressionPinBase
   inputs [
    Metallic {from=MaterialExpressionLinearInterpolate_1, output=""}
    Roughness {from=MaterialExpressionLinearInterpolate_3, output=""}
   ]
   properties {reroutePins [
     Metallic {expression=&self:MaterialExpressionReroute_6}
     Roughness {expression=&self:MaterialExpressionReroute_9}
    ]}
  }
  MaterialExpressionPinBase_10 {
   class=MaterialExpressionPinBase
   inputs []
   properties {reroutePins [
     BaseColor {expression=&self:MaterialExpressionReroute_27}
     BaseColorTexture {expression=&self:MaterialExpressionReroute_29}
     SkinMaskTexture {expression=&self:MaterialExpressionReroute_30}
    ]}
  }
  MaterialExpressionPinBase_11 {
   class=MaterialExpressionPinBase
   inputs [BaseColor {from=MaterialExpressionStaticSwitchParameter_0, output=""}]
   properties {reroutePins [BaseColor {expression=&self:MaterialExpressionReroute_28}]}
  }
  MaterialExpressionPinBase_12 {
   class=MaterialExpressionPinBase
   inputs []
   properties {reroutePins [
     SkinMaskTexture {expression=&self:MaterialExpressionReroute_11}
     BottomNormalBaseTexture {expression=&self:MaterialExpressionReroute_35}
     BottomNormalAccentTexture {expression=&self:MaterialExpressionReroute_36}
    ]}
  }
  MaterialExpressionPinBase_13 {
   class=MaterialExpressionPinBase
   inputs [
    ClearCoat Thickness {from=MaterialExpressionLinearInterpolate_12, output=""}
    ClearCoat Roughness {from=MaterialExpressionLinearInterpolate_18, output=""}
    ClearCoatBottomNormal {from=MaterialExpressionMaterialFunctionCall_7, output=Result}
   ]
   properties {reroutePins [
     ClearCoat Thickness {expression=&self:MaterialExpressionReroute_32}
     ClearCoat Roughness {expression=&self:MaterialExpressionReroute_33}
     ClearCoatBottomNormal {expression=&self:MaterialExpressionReroute_37}
    ]}
  }
  MaterialExpressionPinBase_14 {
   class=MaterialExpressionPinBase
   inputs []
   properties {reroutePins [
     DetailBaseTexture {expression=&self:MaterialExpressionReroute_31}
     DetailAccentTexture {expression=&self:MaterialExpressionReroute_34}
     SkinMaskTexture {expression=&self:MaterialExpressionReroute_16}
    ]}
  }
  MaterialExpressionPinBase_15 {
   class=MaterialExpressionPinBase
   inputs [DetailsMasked {from=MaterialExpressionLinearInterpolate_15, output=""}]
   properties {reroutePins [DetailsMasked {expression=&self:MaterialExpressionReroute_17}]}
  }
  MaterialExpressionPinBase_16 {
   class=MaterialExpressionPinBase
   inputs []
   properties {reroutePins [RGB {expression=&self:MaterialExpressionReroute_42}]}
  }
  MaterialExpressionPinBase_17 {
   class=MaterialExpressionPinBase
   inputs [
    SharpMasksSwitchOutput {from=MaterialExpressionStaticSwitchParameter_1, output=""}
    TextureRawOutput {from=MaterialExpressionReroute_42, output=RGB}
   ]
   properties {reroutePins [
     SharpMasksSwitchOutput {expression=&self:MaterialExpressionReroute_45}
     TextureRawOutput {expression=&self:MaterialExpressionReroute_46}
    ]}
  }
  MaterialExpressionPinBase_18 {
   class=MaterialExpressionPinBase
   inputs []
   properties {reroutePins [
     BaseColorInput {expression=&self:MaterialExpressionReroute_47}
     TextureInput {expression=&self:MaterialExpressionReroute_48}
    ]}
  }
  MaterialExpressionPinBase_19 {
   class=MaterialExpressionPinBase
   inputs [Result {from=MaterialExpressionStaticSwitchParameter_2, output=""}]
   properties {reroutePins [Result {expression=&self:MaterialExpressionReroute_49}]}
  }
  MaterialExpressionPinBase_2 {
   class=MaterialExpressionPinBase
   inputs []
   properties {reroutePins [
     UV {expression=&self:MaterialExpressionReroute_4}
     UV Detail {expression=&self:MaterialExpressionReroute_12}
     UV BottomNormal {expression=&self:MaterialExpressionReroute_41}
     LOD {expression=&self:MaterialExpressionReroute_21}
     BottomNormalFlatness {expression=&self:MaterialExpressionReroute_39}
    ]}
  }
  MaterialExpressionPinBase_20 {
   class=MaterialExpressionPinBase
   inputs []
   properties {reroutePins [
     SkinMaskTexture {expression=&self:MaterialExpressionReroute_55}
     BodyTextures {expression=&self:MaterialExpressionReroute_58}
     EmissiveColor {expression=&self:MaterialExpressionReroute_56}
    ]}
  }
  MaterialExpressionPinBase_21 {
   class=MaterialExpressionPinBase
   inputs [Output {from=MaterialExpressionMultiply_4, output=""}]
   properties {reroutePins [Output {expression=&self:MaterialExpressionReroute_57}]}
  }
  MaterialExpressionPinBase_3 {
   class=MaterialExpressionPinBase
   inputs [
    Curvature {from=MaterialExpressionTextureSampleParameter2D_0, output=R}
    AO {from=MaterialExpressionTextureSampleParameter2D_0, output=R}
    PBR {from=MaterialExpressionAppendVector_0, output=""}
    SkinMask {from=MaterialExpressionTextureSampleParameter2D_1, output=RGB}
    BaseColor {from=MaterialExpressionTextureSampleParameter2D_2, output=RGB}
    DetailBase {from=MaterialExpressionTextureSampleParameter2D_3, output=RGB}
    DetailAccent {from=MaterialExpressionTextureSampleParameter2D_4, output=RGB}
    BodyNormalLOD {from=MaterialExpressionLinearInterpolate_4, output=""}
    BottomNormalBase {from=MaterialExpressionMaterialFunctionCall_1, output=Result}
    BottomNormalAccent {from=MaterialExpressionMaterialFunctionCall_9, output=Result}
   ]
   properties {reroutePins [
     Curvature {expression=&self:MaterialExpressionReroute_22}
     AO {expression=&self:MaterialExpressionReroute_19}
     PBR {expression=&self:MaterialExpressionReroute_53}
     SkinMask {expression=&self:MaterialExpressionReroute_44}
     BaseColor {expression=&self:MaterialExpressionReroute_7}
     DetailBase {expression=&self:MaterialExpressionReroute_0}
     DetailAccent {expression=&self:MaterialExpressionReroute_1}
     BodyNormalLOD {expression=&self:MaterialExpressionReroute_14}
     BottomNormalBase {expression=&self:MaterialExpressionReroute_15}
     BottomNormalAccent {expression=&self:MaterialExpressionReroute_38}
    ]}
  }
  MaterialExpressionPinBase_4 {class=MaterialExpressionPinBase, inputs [], properties {}}
  MaterialExpressionPinBase_5 {
   class=MaterialExpressionPinBase
   inputs [
    TeamColor {from=MaterialExpressionVectorParameter_0, output=RGB}
    AccentColor {from=MaterialExpressionVectorParameter_1, output=RGB}
    PaintColor {from=MaterialExpressionVectorParameter_3, output=RGB}
    TrimColor {from=MaterialExpressionVectorParameter_2, output=RGB}
    EmissiveColor {from=MaterialExpressionVectorParameter_5, output=RGB}
   ]
   properties {reroutePins [
     TeamColor {expression=&self:MaterialExpressionReroute_23}
     AccentColor {expression=&self:MaterialExpressionReroute_24}
     PaintColor {expression=&self:MaterialExpressionReroute_25}
     TrimColor {expression=&self:MaterialExpressionReroute_43}
     EmissiveColor {expression=&self:MaterialExpressionReroute_54}
    ]}
  }
  MaterialExpressionPinBase_6 {
   class=MaterialExpressionPinBase
   inputs []
   properties {reroutePins [
     Roughness {expression=&self:MaterialExpressionReroute_2}
     SkinMask {expression=&self:MaterialExpressionReroute_3}
    ]}
  }
  MaterialExpressionPinBase_7 {
   class=MaterialExpressionPinBase
   inputs [RoughnessMasked {from=MaterialExpressionLinearInterpolate_6, output=""}]
   properties {reroutePins [RoughnessMasked {expression=&self:MaterialExpressionReroute_5}]}
  }
  MaterialExpressionPinBase_8 {
   class=MaterialExpressionPinBase
   inputs []
   properties {reroutePins [
     TeamColor {expression=&self:MaterialExpressionReroute_8}
     AccentColor {expression=&self:MaterialExpressionReroute_13}
     PaintColor {expression=&self:MaterialExpressionReroute_18}
     SkinMaskTexture {expression=&self:MaterialExpressionReroute_20}
    ]}
  }
  MaterialExpressionPinBase_9 {
   class=MaterialExpressionPinBase
   inputs [BaseColor {from=MaterialExpressionLinearInterpolate_8, output=""}]
   properties {reroutePins [BaseColor {expression=&self:MaterialExpressionReroute_26}]}
  }
  MaterialExpressionPower_0 {
   class=MaterialExpressionPower
   inputs [
    Base {from=MaterialExpressionReroute_42, output=RGB}
    Exp {from=""}
   ]
   properties {constExponent=0.4545}
  }
  MaterialExpressionPower_1 {
   class=MaterialExpressionPower
   inputs [
    Base {from=MaterialExpressionReroute_48, output=TextureInput}
    Exp {from=""}
   ]
   properties {constExponent=0.4545}
  }
  MaterialExpressionPower_2 {
   class=MaterialExpressionPower
   inputs [
    Base {from=MaterialExpressionFresnel_0}
    Exp {from=MaterialExpressionScalarParameter_51}
   ]
  }
  MaterialExpressionPower_5 {
   class=MaterialExpressionPower
   inputs [
    Base {from=MaterialExpressionReroute_52}
    Exp {from=""}
   ]
   properties {constExponent=8}
  }
  MaterialExpressionPower_6 {
   class=MaterialExpressionPower
   inputs [
    Base {from=MaterialExpressionClamp_15}
    Exp {from=MaterialExpressionConstant_14}
   ]
  }
  MaterialExpressionPower_7 {
   class=MaterialExpressionPower
   inputs [
    Base {from=MaterialExpressionAdd_4}
    Exp {from=""}
   ]
   properties {constExponent=0.25}
  }
  MaterialExpressionPower_8 {
   class=MaterialExpressionPower
   inputs [
    Base {from=MaterialExpressionComponentMask_15}
    Exp {from=MaterialExpressionScalarParameter_49}
   ]
  }
  MaterialExpressionReflectionVectorWS_0 {class=MaterialExpressionReflectionVectorWS, inputs [CustomWorldNormal {from="", output=""}], properties {}}
  MaterialExpressionReroute_10 {class=MaterialExpressionReroute, inputs [None {from=MaterialExpressionNamedRerouteUsage_12}]}
  MaterialExpressionReroute_11 {class=MaterialExpressionReroute, inputs [None {from=MaterialExpressionNamedRerouteUsage_13}]}
  MaterialExpressionReroute_12 {class=MaterialExpressionReroute, inputs [None {from=MaterialExpressionNamedRerouteUsage_26}]}
  MaterialExpressionReroute_13 {class=MaterialExpressionReroute, inputs [None {from=""}]}
  MaterialExpressionReroute_16 {class=MaterialExpressionReroute, inputs [None {from=MaterialExpressionNamedRerouteUsage_18}]}
  MaterialExpressionReroute_17 {class=MaterialExpressionReroute, inputs [None {from=MaterialExpressionLinearInterpolate_15}]}
  MaterialExpressionReroute_18 {class=MaterialExpressionReroute, inputs [None {from=MaterialExpressionNamedRerouteUsage_10}]}
  MaterialExpressionReroute_2 {class=MaterialExpressionReroute, inputs [None {from=MaterialExpressionNamedRerouteUsage_5}]}
  MaterialExpressionReroute_20 {class=MaterialExpressionReroute, inputs [None {from=MaterialExpressionNamedRerouteUsage_7}]}
  MaterialExpressionReroute_21 {class=MaterialExpressionReroute, inputs [None {from=MaterialExpressionScalarParameter_1}]}
  MaterialExpressionReroute_23 {class=MaterialExpressionReroute, inputs [None {from=MaterialExpressionVectorParameter_0, output=RGB}]}
  MaterialExpressionReroute_24 {class=MaterialExpressionReroute, inputs [None {from=MaterialExpressionVectorParameter_1, output=RGB}]}
  MaterialExpressionReroute_25 {class=MaterialExpressionReroute, inputs [None {from=MaterialExpressionVectorParameter_3, output=RGB}]}
  MaterialExpressionReroute_26 {class=MaterialExpressionReroute, inputs [None {from=MaterialExpressionLinearInterpolate_8}]}
  MaterialExpressionReroute_27 {class=MaterialExpressionReroute, inputs [None {from=MaterialExpressionReroute_26, output=BaseColor}]}
  MaterialExpressionReroute_28 {class=MaterialExpressionReroute, inputs [None {from=MaterialExpressionStaticSwitchParameter_0}]}
  MaterialExpressionReroute_29 {class=MaterialExpressionReroute, inputs [None {from=MaterialExpressionNamedRerouteUsage_15}]}
  MaterialExpressionReroute_3 {class=MaterialExpressionReroute, inputs [None {from=MaterialExpressionNamedRerouteUsage_4}]}
  MaterialExpressionReroute_30 {class=MaterialExpressionReroute, inputs [None {from=MaterialExpressionNamedRerouteUsage_14}]}
  MaterialExpressionReroute_31 {class=MaterialExpressionReroute, inputs [None {from=MaterialExpressionNamedRerouteUsage_29}]}
  MaterialExpressionReroute_32 {class=MaterialExpressionReroute, inputs [None {from=MaterialExpressionLinearInterpolate_12}]}
  MaterialExpressionReroute_33 {class=MaterialExpressionReroute, inputs [None {from=MaterialExpressionLinearInterpolate_18}]}
  MaterialExpressionReroute_34 {class=MaterialExpressionReroute, inputs [None {from=MaterialExpressionNamedRerouteUsage_30}]}
  MaterialExpressionReroute_35 {class=MaterialExpressionReroute, inputs [None {from=MaterialExpressionNamedRerouteUsage_11}]}
  MaterialExpressionReroute_36 {class=MaterialExpressionReroute, inputs [None {from=MaterialExpressionNamedRerouteUsage_19}]}
  MaterialExpressionReroute_37 {class=MaterialExpressionReroute, inputs [None {from=MaterialExpressionMaterialFunctionCall_7, output=Result}]}
  MaterialExpressionReroute_38 {class=MaterialExpressionReroute, inputs [None {from=MaterialExpressionMaterialFunctionCall_9, output=Result}]}
  MaterialExpressionReroute_39 {class=MaterialExpressionReroute, inputs [None {from=MaterialExpressionScalarParameter_14}]}
  MaterialExpressionReroute_4 {class=MaterialExpressionReroute, inputs [None {from=MaterialExpressionNamedRerouteUsage_25}]}
  MaterialExpressionReroute_40 {class=MaterialExpressionReroute, inputs [None {from=MaterialExpressionReroute_39, output=BottomNormalFlatness}]}
  MaterialExpressionReroute_41 {class=MaterialExpressionReroute, inputs [None {from=MaterialExpressionNamedRerouteUsage_21}]}
  MaterialExpressionReroute_42 {class=MaterialExpressionReroute, inputs [None {from=MaterialExpressionReroute_44, output=SkinMask}]}
  MaterialExpressionReroute_43 {class=MaterialExpressionReroute, inputs [None {from=MaterialExpressionVectorParameter_2, output=RGB}]}
  MaterialExpressionReroute_45 {class=MaterialExpressionReroute, inputs [None {from=MaterialExpressionStaticSwitchParameter_1}]}
  MaterialExpressionReroute_46 {class=MaterialExpressionReroute, inputs [None {from=MaterialExpressionReroute_42, output=RGB}]}
  MaterialExpressionReroute_47 {class=MaterialExpressionReroute, inputs [None {from=MaterialExpressionReroute_28, output=BaseColor}]}
  MaterialExpressionReroute_48 {class=MaterialExpressionReroute, inputs [None {from=MaterialExpressionNamedRerouteUsage_27}]}
  MaterialExpressionReroute_49 {class=MaterialExpressionReroute, inputs [None {from=MaterialExpressionStaticSwitchParameter_2}]}
  MaterialExpressionReroute_5 {class=MaterialExpressionReroute, inputs [None {from=MaterialExpressionLinearInterpolate_6}]}
  MaterialExpressionReroute_50 {class=MaterialExpressionReroute, inputs [None {from=MaterialExpressionReroute_47, output=BaseColorInput}]}
  MaterialExpressionReroute_51 {class=MaterialExpressionReroute, inputs [None {from=MaterialExpressionReroute_50}]}
  MaterialExpressionReroute_52 {class=MaterialExpressionReroute, inputs [None {from=MaterialExpressionComponentMask_1}]}
  MaterialExpressionReroute_53 {class=MaterialExpressionReroute, inputs [None {from=MaterialExpressionAppendVector_0}]}
  MaterialExpressionReroute_54 {class=MaterialExpressionReroute, inputs [None {from=MaterialExpressionVectorParameter_5, output=RGB}]}
  MaterialExpressionReroute_55 {class=MaterialExpressionReroute, inputs [None {from=MaterialExpressionNamedRerouteUsage_39}]}
  MaterialExpressionReroute_56 {class=MaterialExpressionReroute, inputs [None {from=MaterialExpressionNamedRerouteUsage_33}]}
  MaterialExpressionReroute_57 {class=MaterialExpressionReroute, inputs [None {from=MaterialExpressionMultiply_4}]}
  MaterialExpressionReroute_58 {class=MaterialExpressionReroute, inputs [None {from=MaterialExpressionNamedRerouteUsage_38}]}
  MaterialExpressionReroute_59 {class=MaterialExpressionReroute, inputs [None {from=MaterialExpressionReroute_56, output=EmissiveColor}]}
  MaterialExpressionReroute_6 {class=MaterialExpressionReroute, inputs [None {from=MaterialExpressionLinearInterpolate_1}]}
  MaterialExpressionReroute_60 {class=MaterialExpressionReroute, inputs [None {from=MaterialExpressionReroute_59}]}
  MaterialExpressionReroute_64 {class=MaterialExpressionReroute, inputs [None {from=MaterialExpressionReroute_65}]}
  MaterialExpressionReroute_65 {class=MaterialExpressionReroute, inputs [None {from=MaterialExpressionReroute_66}]}
  MaterialExpressionReroute_66 {class=MaterialExpressionReroute, inputs [None {from=MaterialExpressionBumpOffset_1}]}
  MaterialExpressionReroute_8 {class=MaterialExpressionReroute, inputs [None {from=""}]}
  MaterialExpressionReroute_9 {class=MaterialExpressionReroute, inputs [None {from=MaterialExpressionLinearInterpolate_3}]}
  MaterialExpressionRotator_0 {
   class=MaterialExpressionRotator
   inputs [
    Coordinate {from=MaterialExpressionAppendVector_9, output=""}
    Time {from=MaterialExpressionDivide_3, output=""}
   ]
   properties {speed=1.75}
  }
  MaterialExpressionRotator_5 {
   class=MaterialExpressionRotator
   inputs [
    Coordinate {from=MaterialExpressionAppendVector_11, output=""}
    Time {from=MaterialExpressionDivide_7, output=""}
   ]
   properties {speed=1.75}
  }
  MaterialExpressionRotator_6 {
   class=MaterialExpressionRotator
   inputs [
    Coordinate {from=MaterialExpressionAppendVector_15, output=""}
    Time {from=MaterialExpressionDivide_14, output=""}
   ]
   properties {}
  }
  MaterialExpressionRotator_7 {
   class=MaterialExpressionRotator
   inputs [
    Coordinate {from=MaterialExpressionAppendVector_16, output=""}
    Time {from=MaterialExpressionDivide_15, output=""}
   ]
   properties {}
  }
  MaterialExpressionSaturate_0 {class=MaterialExpressionSaturate, inputs [None {from=MaterialExpressionConstantBiasScale_1}]}
  MaterialExpressionSaturate_1 {class=MaterialExpressionSaturate, inputs [None {from=MaterialExpressionMaterialFunctionCall_10, output=G}]}
  MaterialExpressionSaturate_2 {class=MaterialExpressionSaturate, inputs [None {from=MaterialExpressionClamp_1}]}
  MaterialExpressionSaturate_4 {class=MaterialExpressionSaturate, inputs [None {from=MaterialExpressionMultiply_8}]}
  MaterialExpressionScalarParameter_0 {class=MaterialExpressionScalarParameter, properties {defaultValue=3, group=Finish, parameterName=DetailTiling}}
  MaterialExpressionScalarParameter_1 {class=MaterialExpressionScalarParameter, properties {parameterName=LOD}}
  MaterialExpressionScalarParameter_10 {class=MaterialExpressionScalarParameter, properties {defaultValue=1, group=Finish, parameterName=CCThickness_PaintColor}}
  MaterialExpressionScalarParameter_11 {class=MaterialExpressionScalarParameter, properties {defaultValue=0.1, group=Finish, parameterName=CCRoughness_TeamColor}}
  MaterialExpressionScalarParameter_12 {class=MaterialExpressionScalarParameter, properties {defaultValue=0.4, group=Finish, parameterName=CCRoughness_AccentColor}}
  MaterialExpressionScalarParameter_13 {class=MaterialExpressionScalarParameter, properties {defaultValue=0.4, group=Finish, parameterName=CCRoughness_PaintColor}}
  MaterialExpressionScalarParameter_14 {class=MaterialExpressionScalarParameter, properties {defaultValue=0.85, group=DetailNormal, parameterName=BottomNormalFlatness}}
  MaterialExpressionScalarParameter_15 {class=MaterialExpressionScalarParameter, properties {defaultValue=13, group=Finish, parameterName=BottomNormalTiling}}
  MaterialExpressionScalarParameter_16 {class=MaterialExpressionScalarParameter, properties {defaultValue=1, group=Color, parameterName=EmissiveBrightness, sortPriority=7}}
  MaterialExpressionScalarParameter_18 {class=MaterialExpressionScalarParameter, properties {defaultValue=0.7, parameterName=Blend}}
  MaterialExpressionScalarParameter_19 {class=MaterialExpressionScalarParameter, properties {defaultValue=1, parameterName=InnerGlowEmissive}}
  MaterialExpressionScalarParameter_2 {class=MaterialExpressionScalarParameter, properties {defaultValue=0.7, group=Finish, parameterName=Metallic_TeamColor}}
  MaterialExpressionScalarParameter_20 {class=MaterialExpressionScalarParameter, properties {defaultValue=1, parameterName=EmissiveIntensity}}
  MaterialExpressionScalarParameter_22 {class=MaterialExpressionScalarParameter, properties {defaultValue=1, parameterName=BaseEmissiveIntensity}}
  MaterialExpressionScalarParameter_23 {class=MaterialExpressionScalarParameter, properties {defaultValue=1, group=Colors, parameterName=SecundaryColorMaskSoftness}}
  MaterialExpressionScalarParameter_24 {class=MaterialExpressionScalarParameter, properties {defaultValue=90, group=Logo, parameterName=Type 1 Rotate}}
  MaterialExpressionScalarParameter_26 {class=MaterialExpressionScalarParameter, properties {defaultValue=270, group=Logo, parameterName=Type 2 Rotate}}
  MaterialExpressionScalarParameter_28 {class=MaterialExpressionScalarParameter, properties {group=Logo, parameterName=LogoRotate}}
  MaterialExpressionScalarParameter_29 {class=MaterialExpressionScalarParameter, properties {defaultValue=1, group=Logo, parameterName=Logo Gamma}}
  MaterialExpressionScalarParameter_3 {class=MaterialExpressionScalarParameter, properties {defaultValue=0.7, group=Finish, parameterName=Metallic_AccentColor}}
  MaterialExpressionScalarParameter_30 {class=MaterialExpressionScalarParameter, properties {group=Logo, parameterName=LogoRotateRSide}}
  MaterialExpressionScalarParameter_4 {class=MaterialExpressionScalarParameter, properties {defaultValue=0.7, group=Finish, parameterName=Metallic_PaintColor}}
  MaterialExpressionScalarParameter_41 {class=MaterialExpressionScalarParameter, properties {defaultValue=-0.05, parameterName=Shimmer1SelfDistort}}
  MaterialExpressionScalarParameter_42 {class=MaterialExpressionScalarParameter, properties {defaultValue=0.03, parameterName=Shimmer2SelfDistort}}
  MaterialExpressionScalarParameter_43 {class=MaterialExpressionScalarParameter, properties {defaultValue=0.75, parameterName=Shimmer1Speed}}
  MaterialExpressionScalarParameter_44 {class=MaterialExpressionScalarParameter, properties {defaultValue=1, parameterName=Shimmer2Speed}}
  MaterialExpressionScalarParameter_45 {class=MaterialExpressionScalarParameter, properties {defaultValue=10, parameterName=SparkleIntensity}}
  MaterialExpressionScalarParameter_46 {class=MaterialExpressionScalarParameter, properties {defaultValue=1, parameterName=ShimmerFresnelFade}}
  MaterialExpressionScalarParameter_47 {class=MaterialExpressionScalarParameter, properties {defaultValue=0.5, parameterName=Shimmer1Power}}
  MaterialExpressionScalarParameter_48 {class=MaterialExpressionScalarParameter, properties {defaultValue=0.2, parameterName=SecundaryColorDarken}}
  MaterialExpressionScalarParameter_49 {class=MaterialExpressionScalarParameter, properties {defaultValue=6, group=InnerColor, parameterName=InnerColorFallof}}
  MaterialExpressionScalarParameter_5 {class=MaterialExpressionScalarParameter, properties {defaultValue=0.4, group=Finish, parameterName=Rough_TeamColor}}
  MaterialExpressionScalarParameter_50 {class=MaterialExpressionScalarParameter, properties {defaultValue=1.2, group=InnerColor, parameterName=InnerCOlorFallofMultiplier}}
  MaterialExpressionScalarParameter_51 {class=MaterialExpressionScalarParameter, properties {defaultValue=0.75, parameterName=FresnelSharpness}}
  MaterialExpressionScalarParameter_52 {class=MaterialExpressionScalarParameter, properties {defaultValue=1, parameterName=Fresnel}}
  MaterialExpressionScalarParameter_53 {class=MaterialExpressionScalarParameter, properties {defaultValue=5, parameterName=FresnelPower}}
  MaterialExpressionScalarParameter_6 {class=MaterialExpressionScalarParameter, properties {defaultValue=0.5, group=Finish, parameterName=Rough_AccentColor}}
  MaterialExpressionScalarParameter_7 {class=MaterialExpressionScalarParameter, properties {defaultValue=0.5, group=Finish, parameterName=Rough_PaintColor}}
  MaterialExpressionScalarParameter_8 {class=MaterialExpressionScalarParameter, properties {defaultValue=1, group=Finish, parameterName=CCThickness_TeamColor}}
  MaterialExpressionScalarParameter_9 {class=MaterialExpressionScalarParameter, properties {defaultValue=1, group=Finish, parameterName=CCThickness_AccentColor}}
  MaterialExpressionSetMaterialAttributes_0 {
   class=MaterialExpressionSetMaterialAttributes
   inputs [
    MaterialAttributes {from="", output=""}
    BaseColor {from=MaterialExpressionNamedRerouteDeclaration_17, output=""}
    Metallic {from=MaterialExpressionNamedRerouteUsage_36, output=""}
    Roughness {from=MaterialExpressionNamedRerouteUsage_37, output=""}
    CustomData0 {from=MaterialExpressionNamedRerouteUsage_22, output=""}
    CustomData1 {from=MaterialExpressionNamedRerouteUsage_23, output=""}
    Normal {from=MaterialExpressionNamedRerouteUsage_3, output=""}
    ClearCoatBottomNormal {from=MaterialExpressionNamedRerouteUsage_32, output=""}
    EmissiveColor {from=MaterialExpressionNamedRerouteUsage_0, output=""}
   ]
   properties {attributeSetTypes [
     #1
     #2
     #3
     #4
     #5
     #6
     #7
     #8
    ]}
  }
  MaterialExpressionStaticSwitchParameter_0 {
   class=MaterialExpressionStaticSwitchParameter
   inputs [
    True {from=MaterialExpressionLinearInterpolate_10, output=""}
    False {from=MaterialExpressionReroute_27, output=BaseColor}
   ]
   properties {parameterName=UseBaseColorTexture}
  }
  MaterialExpressionStaticSwitchParameter_1 {
   class=MaterialExpressionStaticSwitchParameter
   inputs [
    True {from=MaterialExpressionSaturate_0, output=""}
    False {from=MaterialExpressionReroute_42, output=RGB}
   ]
   properties {
    desc="DO NOT USE FOR NEW ASSETS! This is intended for porting content from Rocket League only."
    group=Legacy_PORTINGONLY
    parameterName=SharpMasks
   }
  }
  MaterialExpressionStaticSwitchParameter_12 {
   class=MaterialExpressionStaticSwitchParameter
   inputs [
    True {from=MaterialExpressionMax_8, output=""}
    False {from=MaterialExpressionConstant_28, output=""}
   ]
   properties {defaultValue=true, group=Logo, parameterName=Type Logo}
  }
  MaterialExpressionStaticSwitchParameter_13 {
   class=MaterialExpressionStaticSwitchParameter
   inputs [
    True {from=MaterialExpressionAppendVector_12, output=""}
    False {from=MaterialExpressionAppendVector_13, output=""}
   ]
   properties {group=Logo, parameterName=Side Logos}
  }
  MaterialExpressionStaticSwitchParameter_14 {
   class=MaterialExpressionStaticSwitchParameter
   inputs [
    True {from=MaterialExpressionTextureCoordinate_17, output=""}
    False {from=MaterialExpressionTextureCoordinate_18, output=""}
   ]
   properties {defaultValue=true, desc="True = UV1 ____ False = UV2", group=Logo, parameterName=Logo UV set}
  }
  MaterialExpressionStaticSwitchParameter_15 {
   class=MaterialExpressionStaticSwitchParameter
   inputs [
    True {from=MaterialExpressionLinearInterpolate_44, output=""}
    False {from=MaterialExpressionTextureSampleParameter2D_18, output=RGB}
   ]
   properties {group=Logo, parameterName=Side Logos}
  }
  MaterialExpressionStaticSwitchParameter_2 {
   class=MaterialExpressionStaticSwitchParameter
   inputs [
    True {from=MaterialExpressionMultiply_1, output=""}
    False {from=MaterialExpressionReroute_51, output=""}
   ]
   properties {
    desc="DO NOT USE FOR NEW ASSETS! This is intended for porting content from Rocket League only."
    group=Legacy_PORTINGONLY
    parameterName=SharpMasks
   }
  }
  MaterialExpressionStaticSwitchParameter_3 {
   class=MaterialExpressionStaticSwitchParameter
   inputs [
    True {from=MaterialExpressionTextureCoordinate_4, output=""}
    False {from=MaterialExpressionTextureCoordinate_5, output=""}
   ]
   properties {group=DetailNormal, parameterName=UVChannel2}
  }
  MaterialExpressionStaticSwitchParameter_4 {
   class=MaterialExpressionStaticSwitchParameter
   inputs [
    True {from=MaterialExpressionMultiply_7, output=""}
    False {from=MaterialExpressionConstant_1, output=""}
   ]
   properties {desc=Enables an Emissive Color parameter that is driven by a mask in the BodyTextures texture, group=Color, parameterName=UseEmissiveColor}
  }
  MaterialExpressionStaticSwitchParameter_5 {
   class=MaterialExpressionStaticSwitchParameter
   inputs [
    True {from=MaterialExpressionComponentMask_2, output=""}
    False {from=MaterialExpressionNamedRerouteUsage_28, output=""}
   ]
   properties {
    desc="When using, ensure PBR texture has \"Compress Without Alpha\" set to False. "
    group=Finish
    parameterName=UsePBR_BodyTextures
    sortPriority=-10
   }
  }
  MaterialExpressionStaticSwitchParameter_6 {
   class=MaterialExpressionStaticSwitchParameter
   inputs [
    True {from=MaterialExpressionComponentMask_3, output=""}
    False {from=MaterialExpressionNamedRerouteDeclaration_11, output=""}
   ]
   properties {
    desc="When using, ensure PBR texture has \"Compress Without Alpha\" set to False. "
    group=Finish
    parameterName=UsePBR_BodyTextures
    sortPriority=-10
   }
  }
  MaterialExpressionStaticSwitchParameter_8 {
   class=MaterialExpressionStaticSwitchParameter
   inputs [
    True {from=MaterialExpressionMax_9, output=""}
    False {from=MaterialExpressionTextureSampleParameter2D_18, output=A}
   ]
   properties {group=Logo, parameterName=Side Logos}
  }
  MaterialExpressionSubtract_1 {
   class=MaterialExpressionSubtract
   inputs [
    A {from=MaterialExpressionReroute_58, output=BodyTextures}
    B {from=""}
   ]
   properties {constB=0.098}
  }
  MaterialExpressionSubtract_10 {
   class=MaterialExpressionSubtract
   inputs [
    A {from=MaterialExpressionMultiply_79}
    B {from=MaterialExpressionDivide_1}
   ]
  }
  MaterialExpressionSubtract_11 {
   class=MaterialExpressionSubtract
   inputs [
    A {from=MaterialExpressionMultiply_80}
    B {from=MaterialExpressionDivide_2}
   ]
  }
  MaterialExpressionSubtract_12 {
   class=MaterialExpressionSubtract
   inputs [
    A {from=MaterialExpressionVectorParameter_7, output=R}
    B {from=MaterialExpressionConstant_24}
   ]
  }
  MaterialExpressionSubtract_13 {
   class=MaterialExpressionSubtract
   inputs [
    A {from=MaterialExpressionVectorParameter_7, output=G}
    B {from=MaterialExpressionConstant_24}
   ]
  }
  MaterialExpressionSubtract_14 {
   class=MaterialExpressionSubtract
   inputs [
    A {from=MaterialExpressionTextureCoordinate_16}
    B {from=MaterialExpressionDivide_4}
   ]
  }
  MaterialExpressionSubtract_15 {
   class=MaterialExpressionSubtract
   inputs [
    A {from=MaterialExpressionVectorParameter_9, output=R}
    B {from=MaterialExpressionConstant_30}
   ]
  }
  MaterialExpressionSubtract_16 {
   class=MaterialExpressionSubtract
   inputs [
    A {from=MaterialExpressionVectorParameter_9, output=G}
    B {from=MaterialExpressionConstant_30}
   ]
  }
  MaterialExpressionSubtract_17 {
   class=MaterialExpressionSubtract
   inputs [
    A {from=MaterialExpressionMultiply_82}
    B {from=MaterialExpressionDivide_5}
   ]
  }
  MaterialExpressionSubtract_18 {
   class=MaterialExpressionSubtract
   inputs [
    A {from=MaterialExpressionMultiply_83}
    B {from=MaterialExpressionDivide_6}
   ]
  }
  MaterialExpressionSubtract_19 {
   class=MaterialExpressionSubtract
   inputs [
    A {from=MaterialExpressionStaticSwitchParameter_14}
    B {from=MaterialExpressionDivide_8}
   ]
  }
  MaterialExpressionSubtract_2 {
   class=MaterialExpressionSubtract
   inputs [
    A {from=MaterialExpressionAdd_10}
    B {from=MaterialExpressionMultiply_37}
   ]
  }
  MaterialExpressionSubtract_20 {
   class=MaterialExpressionSubtract
   inputs [
    A {from=MaterialExpressionVectorParameter_13, output=R}
    B {from=MaterialExpressionConstant_34}
   ]
  }
  MaterialExpressionSubtract_21 {
   class=MaterialExpressionSubtract
   inputs [
    A {from=MaterialExpressionVectorParameter_13, output=G}
    B {from=MaterialExpressionConstant_34}
   ]
  }
  MaterialExpressionSubtract_22 {
   class=MaterialExpressionSubtract
   inputs [
    A {from=MaterialExpressionStaticSwitchParameter_14}
    B {from=MaterialExpressionDivide_11}
   ]
  }
  MaterialExpressionSubtract_23 {
   class=MaterialExpressionSubtract
   inputs [
    A {from=MaterialExpressionVectorParameter_13, output=R}
    B {from=MaterialExpressionConstant_37}
   ]
  }
  MaterialExpressionSubtract_24 {
   class=MaterialExpressionSubtract
   inputs [
    A {from=MaterialExpressionVectorParameter_13, output=G}
    B {from=MaterialExpressionConstant_37}
   ]
  }
  MaterialExpressionSubtract_25 {
   class=MaterialExpressionSubtract
   inputs [
    A {from=MaterialExpressionMultiply_84}
    B {from=MaterialExpressionDivide_9}
   ]
  }
  MaterialExpressionSubtract_26 {
   class=MaterialExpressionSubtract
   inputs [
    A {from=MaterialExpressionMultiply_85}
    B {from=MaterialExpressionDivide_10}
   ]
  }
  MaterialExpressionSubtract_27 {
   class=MaterialExpressionSubtract
   inputs [
    A {from=MaterialExpressionMultiply_86}
    B {from=MaterialExpressionDivide_12}
   ]
  }
  MaterialExpressionSubtract_28 {
   class=MaterialExpressionSubtract
   inputs [
    A {from=MaterialExpressionMultiply_87}
    B {from=MaterialExpressionDivide_13}
   ]
  }
  MaterialExpressionSubtract_8 {
   class=MaterialExpressionSubtract
   inputs [
    A {from=MaterialExpressionAdd_3}
    B {from=MaterialExpressionMultiply_44}
   ]
  }
  MaterialExpressionSubtract_9 {
   class=MaterialExpressionSubtract
   inputs [
    A {from=MaterialExpressionTextureCoordinate_16}
    B {from=MaterialExpressionDivide_0}
   ]
  }
  MaterialExpressionTextureCoordinate_0 {class=MaterialExpressionTextureCoordinate}
  MaterialExpressionTextureCoordinate_1 {class=MaterialExpressionTextureCoordinate, properties {coordinateIndex=1}}
  MaterialExpressionTextureCoordinate_12 {class=MaterialExpressionTextureCoordinate, properties {coordinateIndex=1}}
  MaterialExpressionTextureCoordinate_13 {class=MaterialExpressionTextureCoordinate, properties {coordinateIndex=1, vTiling=-1}}
  MaterialExpressionTextureCoordinate_16 {class=MaterialExpressionTextureCoordinate}
  MaterialExpressionTextureCoordinate_17 {class=MaterialExpressionTextureCoordinate, properties {coordinateIndex=1}}
  MaterialExpressionTextureCoordinate_18 {class=MaterialExpressionTextureCoordinate, properties {coordinateIndex=2}}
  MaterialExpressionTextureCoordinate_2 {class=MaterialExpressionTextureCoordinate}
  MaterialExpressionTextureCoordinate_4 {class=MaterialExpressionTextureCoordinate, properties {coordinateIndex=1}}
  MaterialExpressionTextureCoordinate_5 {class=MaterialExpressionTextureCoordinate}
  MaterialExpressionTextureCoordinate_6 {class=MaterialExpressionTextureCoordinate, properties {coordinateIndex=1, uTiling=2, vTiling=2}}
  MaterialExpressionTextureCoordinate_7 {class=MaterialExpressionTextureCoordinate}
  MaterialExpressionTextureSampleParameter2D_0 {
   class=MaterialExpressionTextureSampleParameter2D
   inputs [
    UVs {from="", output=""}
    Apply View MipBias {from="", output=""}
   ]
   properties {
    group=Texture
    parameterName=BodyTextures
    samplerSource=SSM_Wrap_WorldGroupSettings
    samplerType=SAMPLERTYPE_LinearColor
    texture=/VehicleCosmetics/Textures/T_Black_Linear
   }
  }
  MaterialExpressionTextureSampleParameter2D_1 {
   class=MaterialExpressionTextureSampleParameter2D
   inputs [
    UVs {from="", output=""}
    Apply View MipBias {from="", output=""}
   ]
   properties {group=Texture, parameterName=SkinMask, samplerSource=SSM_Wrap_WorldGroupSettings, texture=/VehicleCosmetics/Textures/T_Black}
  }
  MaterialExpressionTextureSampleParameter2D_14 {
   class=MaterialExpressionTextureSampleParameter2D
   inputs [
    UVs {from=MaterialExpressionSubtract_2, output=""}
    Apply View MipBias {from="", output=""}
   ]
   properties {parameterName=TierShimmerMap, texture=&/Skin_Pigskin_ShimmerRGB}
  }
  MaterialExpressionTextureSampleParameter2D_15 {
   class=MaterialExpressionTextureSampleParameter2D
   inputs [
    UVs {from=MaterialExpressionSubtract_8, output=""}
    Apply View MipBias {from="", output=""}
   ]
   properties {parameterName=TierShimmerMap, texture=&Skin_Pigskin_ShimmerRGB}
  }
  MaterialExpressionTextureSampleParameter2D_16 {
   class=MaterialExpressionTextureSampleParameter2D
   inputs [
    UVs {from=MaterialExpressionRotator_0, output=""}
    Apply View MipBias {from="", output=""}
   ]
   properties {group=Logo, parameterName=Type Logo, texture=&/T_Type_GB}
  }
  MaterialExpressionTextureSampleParameter2D_17 {
   class=MaterialExpressionTextureSampleParameter2D
   inputs [
    UVs {from=MaterialExpressionRotator_5, output=""}
    Apply View MipBias {from="", output=""}
   ]
   properties {group=Logo, parameterName=Type Logo, texture=&T_Type_GB}
  }
  MaterialExpressionTextureSampleParameter2D_18 {
   class=MaterialExpressionTextureSampleParameter2D
   inputs [
    UVs {from=MaterialExpressionRotator_6, output=""}
    Apply View MipBias {from="", output=""}
   ]
   properties {group=Logo, parameterName=LogoImage, texture=&/T_Logo_GB}
  }
  MaterialExpressionTextureSampleParameter2D_19 {
   class=MaterialExpressionTextureSampleParameter2D
   inputs [
    UVs {from=MaterialExpressionRotator_7, output=""}
    Apply View MipBias {from="", output=""}
   ]
   properties {group=Logo, parameterName="LogoImage(RSide)", texture=&T_Logo_GB}
  }
  MaterialExpressionTextureSampleParameter2D_2 {
   class=MaterialExpressionTextureSampleParameter2D
   inputs [
    UVs {from=MaterialExpressionReroute_4, output=UV}
    Apply View MipBias {from="", output=""}
   ]
   properties {group=Texture, parameterName=SkinBaseColor, samplerSource=SSM_Wrap_WorldGroupSettings, texture=&T_Black}
  }
  MaterialExpressionTextureSampleParameter2D_3 {
   class=MaterialExpressionTextureSampleParameter2D
   inputs [
    UVs {from=MaterialExpressionReroute_12, output=UV Detail}
    Apply View MipBias {from="", output=""}
   ]
   properties {group=Finish, parameterName=DetailBase, samplerSource=SSM_Wrap_WorldGroupSettings, texture=/VehicleCosmetics/Textures/T_White}
  }
  MaterialExpressionTextureSampleParameter2D_4 {
   class=MaterialExpressionTextureSampleParameter2D
   inputs [
    UVs {from=MaterialExpressionReroute_12, output=UV Detail}
    Apply View MipBias {from="", output=""}
   ]
   properties {group=Finish, parameterName=DetailAccent, samplerSource=SSM_Wrap_WorldGroupSettings, texture=&T_White}
  }
  MaterialExpressionTextureSampleParameter2D_5 {
   class=MaterialExpressionTextureSampleParameter2D
   inputs [
    UVs {from="", output=""}
    Apply View MipBias {from="", output=""}
   ]
   properties {group=Texture, parameterName=BodyNormalLOD, samplerType=SAMPLERTYPE_Normal, texture=/VehicleCosmetics/Textures/T_BlankNormal}
  }
  MaterialExpressionTextureSampleParameter2D_6 {
   class=MaterialExpressionTextureSampleParameter2D
   inputs [
    UVs {from=MaterialExpressionReroute_41, output=UV BottomNormal}
    Apply View MipBias {from="", output=""}
   ]
   properties {
    group=Finish
    parameterName=BottomNormalBase
    samplerType=SAMPLERTYPE_Normal
    texture=/VehicleCosmetics/SharedTextures/T_Wrap_metallicFlakes_Small
   }
  }
  MaterialExpressionTextureSampleParameter2D_7 {
   class=MaterialExpressionTextureSampleParameter2D
   inputs [
    UVs {from=MaterialExpressionReroute_41, output=UV BottomNormal}
    Apply View MipBias {from="", output=""}
   ]
   properties {group=Finish, parameterName=BottomNormalAccent, samplerType=SAMPLERTYPE_Normal, texture=/VehicleCosmetics/Textures/T_BlankNormal_B}
  }
  MaterialExpressionTextureSample_0 {
   class=MaterialExpressionTextureSample
   inputs [
    UVs {from=MaterialExpressionAdd_2}
    Tex {from=""}
    Apply View MipBias {from=""}
   ]
   properties {texture=&/Universal_Noise_67_RGB}
  }
  MaterialExpressionTextureSample_2 {
   class=MaterialExpressionTextureSample
   inputs [
    UVs {from=MaterialExpressionPanner_7}
    Tex {from=""}
    Apply View MipBias {from=""}
   ]
   properties {texture=&Skin_Pigskin_ShimmerRGB}
  }
  MaterialExpressionTextureSample_3 {
   class=MaterialExpressionTextureSample
   inputs [
    UVs {from=MaterialExpressionAppendVector_3}
    Tex {from=""}
    Apply View MipBias {from=""}
   ]
   properties {texture=&/Universal_Noise_Heightmap_17}
  }
  MaterialExpressionTime_4 {class=MaterialExpressionTime}
  MaterialExpressionTime_6 {class=MaterialExpressionTime}
  MaterialExpressionTime_7 {class=MaterialExpressionTime}
  MaterialExpressionTransform_0 {class=MaterialExpressionTransform, inputs [None {from=MaterialExpressionCameraVectorWS_2}]}
  MaterialExpressionVectorParameter_0 {class=MaterialExpressionVectorParameter, properties {group=Color, parameterName=BaseColor, sortPriority=1, defaultValue=<rgba 0 0.2 1 1>}}
  MaterialExpressionVectorParameter_1 {class=MaterialExpressionVectorParameter, properties {group=Color, parameterName=AccentColor, sortPriority=2, defaultValue=<rgba 1 1 1 1>}}
  MaterialExpressionVectorParameter_10 {class=MaterialExpressionVectorParameter, properties {group=Logo, parameterName=Type 2 Offset, defaultValue=<rgba -15 -25 0 1>}}
  MaterialExpressionVectorParameter_11 {class=MaterialExpressionVectorParameter, properties {group=Logo, parameterName=Logo Offset L Side, defaultValue=<rgba 0 14.5 0 1>}}
  MaterialExpressionVectorParameter_12 {class=MaterialExpressionVectorParameter, properties {group=Logo, parameterName=Logo Offset, defaultValue=<rgba 0 14.5 0 1>}}
  MaterialExpressionVectorParameter_13 {
   class=MaterialExpressionVectorParameter
   properties {desc="RB = xy tiling", group=Logo, parameterName=Logo Transform, defaultValue=<rgba 5.25 5.25 0 1.05>}
  }
  MaterialExpressionVectorParameter_14 {class=MaterialExpressionVectorParameter, properties {group=Logo, parameterName=Logo Offset R Side, defaultValue=<rgba 10 14.5 0 1>}}
  MaterialExpressionVectorParameter_15 {
   class=MaterialExpressionVectorParameter
   properties {desc="Scale(RG)Offset(BA)", parameterName=TierShimmerNormalTranformOffset, defaultValue=<rgba 10 10 0 0>}
  }
  MaterialExpressionVectorParameter_16 {class=MaterialExpressionVectorParameter, properties {group=Colors, parameterName=SecondColor, defaultValue=<rgba 1 0.8201 0 1>}}
  MaterialExpressionVectorParameter_2 {
   class=MaterialExpressionVectorParameter
   properties {group=Color, parameterName=TrimColor, sortPriority=4, defaultValue=<rgba 0.03 0.03 0.03 0>}
  }
  MaterialExpressionVectorParameter_3 {class=MaterialExpressionVectorParameter, properties {group=Color, parameterName=PaintColor, sortPriority=3, defaultValue=<rgba 0 1 0 1>}}
  MaterialExpressionVectorParameter_4 {class=MaterialExpressionVectorParameter, properties {group=Colors, parameterName=MainColor, defaultValue=<rgba 0 0.1413 0.0953 1>}}
  MaterialExpressionVectorParameter_5 {class=MaterialExpressionVectorParameter, properties {group=Color, parameterName=EmissiveColor, sortPriority=5, defaultValue=<rgba 0 1 1 1>}}
  MaterialExpressionVectorParameter_6 {class=MaterialExpressionVectorParameter, properties {group=Colors, parameterName=ThirdColor, defaultValue=<rgba 1 1 1 1>}}
  MaterialExpressionVectorParameter_7 {
   class=MaterialExpressionVectorParameter
   properties {desc="RG= Scale", group=Logo, parameterName=Type 1 Transform, defaultValue=<rgba 6 6 15 -25>}
  }
  MaterialExpressionVectorParameter_8 {class=MaterialExpressionVectorParameter, properties {group=Logo, parameterName=Type 1 Offset, defaultValue=<rgba 15 -25 0 1>}}
  MaterialExpressionVectorParameter_9 {
   class=MaterialExpressionVectorParameter
   properties {desc="RG= Scale", group=Logo, parameterName=Type 2 Transform, defaultValue=<rgba 6 6 15 -25>}
  }
 }}
